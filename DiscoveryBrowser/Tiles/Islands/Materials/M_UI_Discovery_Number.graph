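# Translucent UI material for numbers on Discovery browser island tiles, judging by its name and folder.
# It samples Slate post buffer 0 through the engine's GetSlatePost0 function and multiplies the linear RGB by vertex color.
# Reading the post buffer lets the element draw on processed screen content, tinted per widget, instead of using its own texture.
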Material /DiscoveryBrowser/Tiles/Islands/Materials/M_UI_Discovery_Number
graph {
 properties {
  bAutomaticallySetUsageInEditor=false
  bUsedWithStaticMesh=false
  blendMode=BLEND_Translucent
  materialDomain=MD_UI
  pixelDepthOffsetMode=PDOM_Legacy
  usedShadingModels=Default Lit
 }
 nodes [Multiply(A=MaterialFunctionCall!("UVs (Optional)") {materialFunction=/Engine/Functions/UserInterface/GetSlatePost0}.LinearRGB, B=VertexColor())]
}
