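# Translucent UI material for number outlines on Discovery browser island tiles, judging by its name and path.
# Multiplies the shared CoreUI BackplateA function result by vertex color for the color.
# Builds a horizontal fade from inverted 9-slice U, remapped, clamped and eased out cubically, so alpha falls off toward the right.
# The Display Mask parameter blends between plain vertex alpha and that faded alpha.
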Material /DiscoveryBrowser/Tiles/Islands/Materials/M_UI_Discovery_NumberOutline
graph {
 properties {
  bAutomaticallySetUsageInEditor=false
  bUsedWithStaticMesh=false
  blendMode=BLEND_Translucent
  materialDomain=MD_UI
  pixelDepthOffsetMode=PDOM_Legacy
  usedShadingModels=Default Lit
 }
 nodes [
  n1 = VertexColor()
  n2 = ComponentMask(None=MaterialFunctionCall() {materialFunction=&GetUserInterfaceUV}.9-Slice UV) {r=true}
  n3 = Saturate(None=MaterialFunctionCall(Input=OneMinus(None=@n2), Input Low=Constant() {r=0.395}, Input High=Constant() {r=1}) {materialFunction=&RemapValue_0-1}.Result)
  LinearInterpolate(
   A=@n1.A
   B=Multiply(A=@n1.A, B=MaterialFunctionCall("Value (0-1)"=@n3) {materialFunction=&MF_UI_EaseCurves}.EaseOutCubic)
   Alpha=ScalarParameter() {parameterName=Display Mask}
  )
  Multiply(A=MaterialFunctionCall!(Post Buffer Blur) {materialFunction=/FortUILibrary/MaterialFunctions/CoreUI/MF_UI_CoreUI_BackplateA}.Result, B=@n1)
 ]
}
