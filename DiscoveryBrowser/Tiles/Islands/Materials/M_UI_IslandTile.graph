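# Translucent UI material for Discovery browser island tiles, drawing a thumbnail with loading, empty and hover states.
# The thumbnail uses scale-to-fill UVs with zoom and optional fixed media size, plus a darkening overlay and a switchable SDF error-icon empty state.
# Download Ratio drives a progress darkening mask, while an animated gradient marks loading.
# A rounded-corner mask shapes opacity, and an eased gradient darkens the tile for hover.
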Material /DiscoveryBrowser/Tiles/Islands/Materials/M_UI_IslandTile
graph {
 properties {
  bUsedWithStaticMesh=false
  blendMode=BLEND_Translucent
  materialDomain=MD_UI
  pixelDepthOffsetMode=PDOM_Legacy
  usedShadingModels=Default Lit
 }
 nodes [
  Desaturation()
  n1 = Constant3Vector() {constant=<rgba 0.3515 0.3515 0.3515 1>}
  n2 = NamedRerouteUsage()
  n3 = LinearInterpolate(A=@n2, B=Multiply(A=Desaturation(None=@n2), B=Constant() {desc=Darken Maximum Amount, r=0.5}), Alpha=NamedRerouteUsage())
  n4 = Multiply(A=NamedRerouteUsage(), B=ScalarParameter() {defaultValue=1, group=1 - Thumbnail, parameterName=Thumbnail Alpha, sliderMax=1, sortPriority=1})
  n5 = LinearInterpolate(A=NamedRerouteUsage(), Alpha=ScalarParameter() {group=1 - Thumbnail, parameterName=Override Loading Visiblity, sliderMax=1, sortPriority=1}) {constB=0}
  n6 = Multiply(A=NamedRerouteUsage(), B=@n5)
  n7 = NamedRerouteDeclaration(None=LinearInterpolate(A=LinearInterpolate(A=@n1, B=@n3, Alpha=@n4), B=@n1, Alpha=Multiply(A=@n6) {constB=0.1})) {name=Thumbnail RGB, nodeColor=<rgba 0 0 0 1>}
  n8 = MaterialFunctionCall("Value (0-1)"=ComponentMask(None=NamedRerouteUsage()) {g=true}) {materialFunction=&MF_UI_EaseCurves}
  n9 = ScalarParameter() {defaultValue=1, parameterName=Hover Overlay Opacity, sliderMax=1}
  LinearInterpolate(
   A=@n7
   B=LinearInterpolate(A=@n7, B=Multiply(A=@n7) {constB=0.01}, Alpha=Multiply(A=@n8.EaseOutSine, B=@n9)) {constB=0}
   Alpha=ScalarParameter() {parameterName=Use Hover Overlay, sliderMax=1}
  )
  LinearInterpolate(Alpha=ScalarParameter() {defaultValue=1, parameterName=Saturation})
  n10 = TextureSample(
   UVs=MaterialFunctionCall(UVs=NamedRerouteUsage(), Texture Scale=ScalarParameter() {defaultValue=1, group=1 - Thumbnail, parameterName=Thumbnail Zoom, sortPriority=2}) {materialFunction=&ScaleUVsByCenter}.UVs
   Tex=NamedRerouteDeclaration(None=TextureObjectParameter() {group=1 - Thumbnail, parameterName=Thumbnail, sortPriority=0}) {name=Thumbnail, nodeColor=<rgba 0 0 0 1>}
  )
  n11 = StaticSwitchParameter(
   True=LinearInterpolate(A=@n10.RGBA, B=NamedRerouteUsage(), Alpha=ScalarParameter() {defaultValue=1, group=2 - Empty, parameterName=Is Empty, sliderMax=1})
   False=@n10.RGBA
  ) {group=2 - Empty, parameterName=Supports Empty State}
  n12 = ScalarParameter() {group=1 - Thumbnail, parameterName=Overlay Opacity, sliderMax=1, sortPriority=3}
  n13 = Constant() {desc=Darken Maximum Amount, r=0.7}
  NamedRerouteDeclaration(None=LinearInterpolate(
    A=ComponentMask(None=@n11) {b=true, g=true, r=true}
    B=Constant3Vector() {desc=Darken Color, constant=<rgba 0 0 0 1>}
    Alpha=Multiply(A=@n12, B=@n13)
   )) {name=Thumbnail Color, nodeColor=<rgba 0 0 0 1>}
  NamedRerouteDeclaration(None=MaterialFunctionCall() {materialFunction=&GetUserInterfaceUV}.Normalized UV) {name=Normalized UV, nodeColor=<rgba 0 0 0 1>}
  n14 = AppendVector(
   A=ScalarParameter() {defaultValue=16, group=1 - Thumbnail, parameterName=Media Width}
   B=ScalarParameter() {defaultValue=9, group=1 - Thumbnail, parameterName=Media Height, sortPriority=33}
  ) {desc=HACK - Used to send the texture size manually to have both image and video to be the same size}
  n15 = TextureProperty(Tex=NamedRerouteUsage())
  NamedRerouteDeclaration(None=MaterialFunctionCall(
    UV=NamedRerouteUsage()
    TextureSize=StaticSwitchParameter(True=@n14, False=@n15) {group=1 - Thumbnail, parameterName=Use Fixed Size, sortPriority=31}
   ) {materialFunction=&MF_UI_UVScaling}.Scaled To Fill) {name=Scale to Fill UV, nodeColor=<rgba 0 0 0 1>}
  NamedRerouteDeclaration(None=ComponentMask(None=@n11) {a=true}) {name=Thumbnail Alpha, nodeColor=<rgba 0 0 0 1>}
  n16 = Constant() {r=2}
  n17 = Constant() {r=0.125}
  n18 = Constant() {r=0.75}
  n19 = MaterialFunctionCall(
   GradientEnd=MaterialFunctionCall("Value (0-1)"=MaterialFunctionCall!(Animation Duration=@n16, Delay=@n17, Reset to Default) {materialFunction=/FortUILibrary/MaterialFunctions/Time/MF_UI_AnimationDelay}.Result) {materialFunction=&MF_UI_EaseCurves}.EaseInOutSine
   RotationAngle=Subtract(A=@n18) {constB=0.75}
  ) {materialFunction=&MF_UI_LinearGradient}
  n20 = MaterialFunctionCall!(A=@n19.Result, B=Constant() {r=1}) {materialFunction=/FortUILibrary/MaterialFunctions/Basics/MF_UI_BooleanOperations}
  NamedRerouteDeclaration(None=Multiply(A=@n19.Result, B=OneMinus(None=@n20.IsEqual))) {name=Loading Mask, nodeColor=<rgba 0 0 0 1>}
  n21 = MaterialFunctionCall(
   A=ScalarParameter() {defaultValue=1, parameterName=Download Ratio, sliderMax=1}
   B=ScalarParameter() {defaultValue=1, group=1 - Thumbnail, parameterName=Thumbnail Alpha, sliderMax=1, sortPriority=1}
  ) {materialFunction=&MF_UI_BooleanOperations}
  n22 = Constant() {r=1}
  NamedRerouteDeclaration(None=OneMinus(None=MaterialFunctionCall(A=@n21.And, B=@n22) {materialFunction=&MF_UI_BooleanOperations}.IsEqual)) {name=Display Loading Mask, nodeColor=<rgba 0 0 0 1>}
  n23 = Rotator(Coordinate=NamedRerouteUsage(), Time=@n18) {speed=6.2832}
  n24 = Subtract(A=OneMinus(None=ComponentMask(None=@n23) {g=true}), B=ScalarParameter() {defaultValue=1, parameterName=Download Ratio, sliderMax=1}) {constB=0.25}
  NamedRerouteDeclaration(None=Ceil(None=Saturate(None=@n24))) {name=Desaturation Mask, nodeColor=<rgba 0 0 0 1>}
  n25 = VectorParameter() {parameterName=Corner Radius, defaultValue=<rgba 20 20 20 20>}
  n26 = MaterialFunctionCall(
   Corner Radius=AppendVector(A=@n25.RGB, B=@n25.RGB)
   Radius in Pixels=ScalarParameter() {defaultValue=1, parameterName=RadiusInPixels}
   Outline Thickness=Constant() {r=4}
  ) {materialFunction=&MF_UI_CoreUI_RoundedMask}
  NamedRerouteDeclaration(None=@n26.Outline Mask) {name=Outline, nodeColor=<rgba 0 0 0 1>}
  n27 = NamedRerouteUsage()
  n28 = NamedRerouteDeclaration(None=TextureSample(
    UVs=MaterialFunctionCall(UV=@n27) {materialFunction=&MF_UI_UVScaling}.Scaled to Fit
    Tex=TextureObject() {samplerType=SAMPLERTYPE_LinearColor, texture=/FortUILibrary/Textures/Icons/T_UI_Icon_UIKit_ImageError_SDF}
   ) {samplerSource=SSM_Clamp_WorldGroupSettings}.R) {name=Empty Image, nodeColor=<rgba 0 0 0 1>}
  n29 = Constant() {r=0.68}
  NamedRerouteDeclaration(None=LinearInterpolate(Alpha=MaterialFunctionCall(DistanceField=@n28, Threshold=@n29) {materialFunction=&MF_UI_AntiAliasing}.Outer) {constA=0.1, constB=0.4}) {name=Empty Thumbnail, nodeColor=<rgba 0 0 0 1>}
  n30 = LinearInterpolate(Alpha=@n5) {constA=0.2, constB=0.4}
  NamedRerouteDeclaration(None=Multiply(A=NamedRerouteDeclaration(None=@n26.Shape Mask) {name=Fill, nodeColor=<rgba 0 0 0 1>}, B=LinearInterpolate(A=@n30, Alpha=@n4) {constA=0.2})) {name=Thumbnail Opacity, nodeColor=<rgba 0 0 0 1>}
  NamedRerouteUsage()
 ]
}
