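# UI material for the selected-game state of Discovery Browser island tiles, drawing a rounded diamond marker with a softly pulsing glow.
# MF_UI_Box draws the shape from UVs rotated a constant 45 degrees.
# Its SDF is remapped, inverted and cubic-eased into a faint falloff, combined with the solid fill by Max.
# A Time-driven triangle wave slightly varies the falloff width over a two-second loop.
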Material /DiscoveryBrowser/Tiles/Islands/Materials/M_UI_IslandTile_State_SelectedGame
graph {
 properties {
  bAutomaticallySetUsageInEditor=false
  bUsedWithStaticMesh=false
  blendMode=BLEND_AlphaComposite
  materialDomain=MD_UI
  usedShadingModels=Default Lit
 }
 nodes [
  n1 = MaterialFunctionCall(
   UVs=Rotator(
    Coordinate=MaterialFunctionCall() {materialFunction=&GetUserInterfaceUV}.Normalized UV
    Time=Constant() {r=0.125}
   ) {speed=6.2832}
   Size=Constant() {r=0.375}
   CornerRoundness=Constant() {r=0.2}
  ) {materialFunction=&MF_UI_Box}
  n2 = Frac(None=Multiply(A=Time()) {constB=0.5})
  n3 = Multiply(A=Min(A=@n2, B=OneMinus(None=@n2))) {constB=2}
  n4 = OneMinus(None=MaterialFunctionCall(Input=Saturate(None=@n1.SDF), Input Low=Constant(), Input High=LinearInterpolate(Alpha=@n3) {constA=0.7, constB=0.75}) {materialFunction=&RemapValue_0-1}.Result)
  Max(
   A=@n1.Fill
   B=Multiply(A=MaterialFunctionCall("Value (0-1)"=@n4) {materialFunction=&MF_UI_EaseCurves}.EaseInCubic) {constB=0.1}
  )
 ]
}
