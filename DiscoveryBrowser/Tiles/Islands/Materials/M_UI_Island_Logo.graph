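# Translucent UI material for an island logo in Discovery Browser tiles.
# It samples a Logo texture parameter, defaulting to the By Epic badge, at the normalized UI UV from GetUserInterfaceUV.
# The sampler uses the shared world-group clamp setting.
# The graph shows no connection from that sample to a material output.
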Material /DiscoveryBrowser/Tiles/Islands/Materials/M_UI_Island_Logo
graph {
 properties {
  bUsedWithStaticMesh=false
  blendMode=BLEND_Translucent
  materialDomain=MD_UI
  pixelDepthOffsetMode=PDOM_Legacy
  usedShadingModels=Default Lit
 }
 nodes [TextureSample(
   UVs=MaterialFunctionCall() {materialFunction=&GetUserInterfaceUV}.Normalized UV
   Tex=TextureObjectParameter() {parameterName=Logo, texture=/DiscoveryBrowser/Tiles/Textures/T_UI_ByEpicBadge}
  ) {samplerSource=SSM_Clamp_WorldGroupSettings}]
}
