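# Contextual button widget for Discovery Browser tiles, derived from WBP_Discover_Button_Generic with no widgets of its own.
# On initialization and on every input method change, ToggleTouchStyle checks the CommonInputSubsystem for touch input.
# It hides the inherited background and main named slots under touch and shows them as self-hit-test-invisible otherwise.
# PreConstruct, Construct and hover events only forward to the parent.
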WidgetBlueprint /DiscoveryBrowser/Tiles/Islands/WBP_Discovery_ContextualButton
graphs [
 EventGraph {nodes [
   n1 = Event PreConstruct()
   > Parent: PreConstruct(IsDesignTime=@n1.IsDesignTime)
   Event Construct()
   > Parent: Construct()
   n2 = Event OnInitialized()
   ToggleTouchStyle(execute=[@n2.then, @n3.then])
   Event On Hovered()
   > Parent: On Hovered() {is=title K2Node_CallParentFunction}
   Event On Unhovered()
   > Parent: On Unhovered() {is=title K2Node_CallParentFunction}
   n3 = Event On Input Method Changed!() -> (OutputDelegate, then, CurrentInputType:ECommonInputType Enum) {is=title K2Node_Event}
  ]}
 ToggleTouchStyle {nodes [
   ToggleTouchStyle() {is=title K2Node_FunctionEntry}
   n1 = "Equal (Enum)"(A=GetCurrentInputType(self=Get CommonInputSubsystem()), B=Touch)
   > SetVisibility(
    self=Get Slot_Background() -> (Slot_Background:Named Slot Object Reference)
    InVisibility=Select(Option 0=SelfHitTestInvisible, Option 1=Hidden, Index=@n1)
   )
   > SetVisibility(
    self=Get Slot_Main() -> (Slot_Main:Named Slot Object Reference)
    InVisibility=Select(Option 0=SelfHitTestInvisible, Option 1=Hidden, Index=@n1)
   )
  ]}
]
variables []
widgets {
 info {
  inheritedWidgetCount=0
  parentClass=/DiscoveryBrowser/Tiles/WBP_Discover_Button_Generic.WBP_Discover_Button_Generic_C
  rootWidgetClass=null
  widgetCount=0
 }
 widgets []
}
