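# Widget for a large outlined number on Discovery browser island tiles, apparently a rank or position label.
# A fixed-size spacer sets the footprint, while a right-aligned CommonVisualAttachment holds the text, styled with fill and outline materials.
# Two PostBufferUpdate widgets with blur updaters bracket the text on the same post buffer, named for lowering blur and then returning to normal.
# The event graph holds only unconnected stub events, and no variables set the number here.
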WidgetBlueprint /DiscoveryBrowser/Tiles/Islands/WBP_Discovery_Number
graphs [EventGraph {nodes [
   Event PreConstruct()
   Event Construct()
   Event Tick()
  ]}]
variables []
widgets {
 info {inheritedWidgetCount=0, parentClass=&UserWidget, rootWidgetClass=&Overlay, widgetCount=6}
 widgets [
  Overlay_48 {widgetClassPath=&Overlay, widget=Overlay_48, properties {visibility=HitTestInvisible}}
  SpacerWidth {
   widgetClassPath=&Spacer
   bIsVariable=true
   parent=Overlay_48
   widget=SpacerWidth
   properties {size=<56 80>}
   slot {class=OverlaySlot, properties {verticalAlignment=VAlign_Fill}}
  }
  Buffer_LowerBlur {
   widgetClassPath=/Script/UMG.PostBufferUpdate
   bIsVariable=true
   parent=Overlay_48
   widget=Buffer_LowerBlur
   properties {
    visibility=HitTestInvisible
    updateBufferInfos [{
      bufferToUpdate=ESlatePostRT_0
      postParamUpdater=&self:WidgetTree.Buffer_LowerBlur.PostBufferBlurUpdater_0
     }]
   }
   slot {class=OverlaySlot, properties {horizontalAlignment=HAlign_Fill, verticalAlignment=VAlign_Fill, padding=<ltrb -16 0 -56 0>}}
  }
  CommonVisualAttachment_0 {
   widgetClassPath=&CommonVisualAttachment
   parent=Overlay_48
   widget=CommonVisualAttachment_0
   properties {contentAnchor=<0.5 0.5>}
   slot {class=OverlaySlot, properties {horizontalAlignment=HAlign_Right, verticalAlignment=VAlign_Center}}
  }
  Text_Number {
   widgetClassPath=&CommonTextBlock
   bIsVariable=true
   parent=CommonVisualAttachment_0
   widget=Text_Number
   properties {
    text=1
    font {
     bForceMonospaced=false
     bMaterialIsStencil=false
     fontMaterial=/DiscoveryBrowser/Tiles/Islands/Materials/MI_UI_Discovery_Number
     fontObject=&HeadingNow
     letterSpacing=-40
     monospacedWidth=1
     size=64.5
     skewAmount=0
     typefaceFontName=86 Bold
     outlineSettings {
      bApplyOutlineToDropShadows=false
      bMiteredCorners=false
      bSeparateFillAlpha=false
      outlineMaterial=/DiscoveryBrowser/Tiles/Islands/Materials/MI_UI_Discovery_NumberOutline
      outlineSize=2
      outlineColor=<rgba 1 1 1 1>
     }
    }
    shadowOffset=<0 0>
   }
   slot {class=SizeBoxSlot, properties {padding=<ltrb 0 0 16 0>}}
  }
  Buffer_BackToNormal {
   widgetClassPath=&PostBufferUpdate
   bIsVariable=true
   parent=Overlay_48
   widget=Buffer_BackToNormal
   properties {
    visibility=HitTestInvisible
    updateBufferInfos [{
      bufferToUpdate=ESlatePostRT_0
      postParamUpdater=&self:WidgetTree.Buffer_BackToNormal.PostBufferBlurUpdater_0
     }]
   }
   slot {class=OverlaySlot, properties {horizontalAlignment=HAlign_Fill, verticalAlignment=VAlign_Fill, padding=<ltrb -16 0 -56 0>}}
  }
 ]
}
