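# Video widget for Discovery browser island tiles, built on FortVideoWidgetBase with the video image filling an overlay through a tile material.
# A short outro animates that material's Thumbnail Alpha down to zero with a slight Thumbnail Zoom increase.
# PlayOutro either animates it or jumps to the end state, and Reset uses reverse playback to restore the start state.
# On initialization it sets the base class's media pre-end time.
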WidgetBlueprint /DiscoveryBrowser/Tiles/Islands/WBP_IslandTile_VideoV2
animations [VideoSwapActivityOutro {
  binding_count=1
  length_seconds=0.24
  object=&self:VideoSwapActivityOutro
  sequence {bindings [Image_VideoTexture {
     id=#1
     possessed_class=Image
     tracks [{
       type=MovieSceneWidgetMaterialTrack
       sections [{
         type=MovieSceneParameterSection
         end_frame=4
         start_frame=0
         channels [
          Thumbnail Alpha {type=MovieSceneScriptingFloatChannel, keys [[0,1],[4.8,0]]}
          Thumbnail Zoom {type=MovieSceneScriptingFloatChannel, keys [[0,0],[4.8,0.05]]}
         ]
        }]
      }]
    }]}
 }]
graphs [EventGraph {nodes [
   n1 = PlayOutro!() -> (OutputDelegate, then, PlayAnimation:bool=true) {is=title K2Node_CustomEvent}
   n2 = Get VideoSwapActivityOutro() -> (VideoSwapActivityOutro:Widget Animation Object Reference)
   > PlayAnimation(InAnimation=@n2, StartAtTime=SelectFloat(A=0, B=GetEndTime(self=@n2), bPickA=@n1.PlayAnimation))
   Reset() {is=title K2Node_CustomEvent}
   n3 = Get VideoSwapActivityOutro()
   > PlayAnimation(InAnimation=@n3, StartAtTime=GetEndTime(self=@n3), PlayMode=Reverse)
   Event OnInitialized()
   > SetMediaPreEndTimeSeconds!(execute, self:Fort Video Widget Base Object Reference, InSeconds:float=0.4) -> (then)
  ]}]
variables []
widgets {
 info {inheritedWidgetCount=4, parentClass=/Script/FortniteUI.FortVideoWidgetBase, rootWidgetClass=&Overlay, widgetCount=5}
 widgets [
  Overlay_0 {widgetClassPath=&Overlay, widget=Overlay_0}
  Image_VideoTexture {
   widgetClassPath=&Image
   bInherited=true
   bIsVariable=true
   parent=Overlay_0
   widget=Image_VideoTexture
   properties {
    visibility=SelfHitTestInvisible
    brush {
     bIsDynamicallyLoaded=false
     drawAs=Image
     imageType=NoImage
     mirroring=NoMirror
     resourceName=None
     resourceObject=/DiscoveryBrowser/Tiles/Islands/Materials/MI_UI_IslandTile_Video
     tiling=NoTile
     imageSize=<32 32>
     margin=<ltrb 0.5 0 0.5 0>
     outlineSettings {
      bUseBrushTransparency=false
      roundingType=HalfHeightRadius
      width=0
      color {colorUseRule=UseColor_Specified, specifiedColor=<rgba 0 0 0 0>}
      cornerRadii=<0 0 0 0>
     }
     tintColor {colorUseRule=UseColor_Specified, specifiedColor=<rgba 1 1 1 1>}
     uVRegion {bIsValid=false, max=<0 0>, min=<0 0>}
    }
   }
   slot {class=OverlaySlot, properties {horizontalAlignment=HAlign_Fill, verticalAlignment=VAlign_Fill}}
  }
  Subtitles {widgetClassPath=/Script/SubtitlesWidgets.SubtitleDisplay, bInherited=true, widget=null}
  Button_Skip {widgetClassPath=&CommonButtonBase, bInherited=true, widget=null}
  Image_Throbber {widgetClassPath=&Image, bInherited=true, widget=null}
 ]
}
