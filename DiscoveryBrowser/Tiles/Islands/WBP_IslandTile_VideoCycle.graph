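# Island tile widget, derived from FortActivityVideoCycle, that fades the selected activity's video in and out, with an optional bumper overlay.
# For unlocked, age-rated activities, PlayVideoBumper fills a Squeegee injection slot, fades it in and calls PlaySelectedActivityMovie.
# A two-second timer then marks the bumper finished, reverses its fade and calls PlayVideo.
# Construct snaps the video fade to its reversed end; Destruct clears the timer and hides the bumper.
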WidgetBlueprint /DiscoveryBrowser/Tiles/Islands/WBP_IslandTile_VideoCycle
animations [
 AnimVideoOpacity {
  binding_count=1
  length_seconds=0.4
  object=&self:AnimVideoOpacity
  sequence {bindings [ActivityMovieWidget {
     id=#1
     possessed_class=WBP_IslandTile_Video_C
     tracks [{
       type=MovieSceneFloatTrack
       display_name=Render Opacity
       sections [{type=MovieSceneFloatSection, end_frame=8, start_frame=0, channels [None {type=MovieSceneScriptingFloatChannel, keys [[0,0],[8,1]]}]}]
      }]
    }]}
 }
 FadeInBumper {
  binding_count=1
  length_seconds=0.3953
  object=&self:FadeInBumper
  sequence {bindings [SqueegeeSlot_Bumper {
     id=#2
     possessed_class=SqueegeeInjectionSlot_C
     tracks [{
       type=MovieSceneFloatTrack
       display_name=Render Opacity
       sections [{type=MovieSceneFloatSection, end_frame=7, start_frame=0, channels [None {type=MovieSceneScriptingFloatChannel, keys [[0,0],[7.9063,1]]}]}]
      }]
    }]}
 }
]
graphs [
 EventGraph {nodes [
   FadeInVideoBumper() {is=title K2Node_CustomEvent}
   > Set bBumperFinished(bBumperFinished:bool=false) -> (Output_Get:bool)
   > n1 = "Play Animation with Finished event (legacy)"!(
    execute:Exec
    Widget:User Widget Object Reference=Self-Reference()
    InAnimation:Widget Animation Object Reference=Get FadeInBumper() -> (FadeInBumper:Widget Animation Object Reference)
    StartAtTime:float=0
    NumLoopsToPlay:int=1
    PlayMode:EUMGSequencePlayMode Enum
    PlaybackSpeed:float=1
   ) -> (then:Exec, Result:UMGSequence Player Object Reference, Finished:Exec) {is=title K2Node_PlayAnimation}
   n2 = DelayedPlayVideo() {is=title K2Node_CustomEvent}
   n3 = Set Timer by Event(execute=@n1.then, Delegate=@n2, Time=2, bLooping=false)
   > Set VideoBumperTimer(VideoBumperTimer:Timer Handle Structure=@n3) -> (Output_Get:Timer Handle Structure)
   > PlaySelectedActivityMovie(self:Fort Activity Video Cycle Object Reference)
   FadeOutVideoBumper() {is=title K2Node_CustomEvent}
   > "Play Animation with Finished event (legacy)"(Widget=Self-Reference(), InAnimation=Get FadeInBumper(), StartAtTime=0, NumLoopsToPlay=1, PlayMode=Reverse, PlaybackSpeed=1)
   Event Destruct()
   > Clear and Invalidate Timer by Handle(Handle=Get VideoBumperTimer() -> (VideoBumperTimer:Timer Handle Structure))
   > SetRenderOpacity(self=Get SqueegeeSlot_Bumper() -> (SqueegeeSlot_Bumper:Squeegee Injection Slot Object Reference), InOpacity=0)
   > Set bIntroHasPlayed(bIntroHasPlayed:bool=false) -> (Output_Get:bool)
   Event PlayVideoIntro() {is=title K2Node_Event}
   > PlayAnimationForward(InAnimation=Get AnimVideoOpacity() -> (AnimVideoOpacity:Widget Animation Object Reference))
   Event Construct()
   > SnapToEndOfAnimation!(execute, self, InAnimation:Widget Animation Object Reference=Get AnimVideoOpacity(), PlayMode:EUMGSequencePlayMode Enum=Reverse) -> (then)
   Event PlayVideoOutro() {is=title K2Node_Event}
   > PlayAnimationReverse(InAnimation=Get AnimVideoOpacity())
   Event PlayVideoBumper() {is=title K2Node_Event}
   n4 = GetSelectedGameActivity!(self:Fort Activity Video Cycle Object Reference) -> (ReturnValue:Fort Game Activity Object Reference)
   n5 = IsLocked!(self:Fort Game Activity Object Reference=@n4) -> (ReturnValue:bool)
   n6 = GetOwningLocalPlayer!(self:Widget Object Reference) -> (ReturnValue:Local Player Object Reference)
   > Branch(Condition=AND Boolean(
     A=HasAgeRating!(self:Fort Game Activity Object Reference=@n4, PlayerContext:Local Player Object Reference=@n6) -> (ReturnValue:bool)
     B=NOT Boolean(A=@n5)
    ))
   > ShowActivityData!(
    execute
    self:Fort Squeegee Widget Injection Slot Object Reference=Get SqueegeeSlot_Bumper()
    GameActivity:Fort Game Activity Object Reference=GetSelectedGameActivity()
   ) -> (then)
   > FadeInVideoBumper() {is=title K2Node_CallFunction}
   PlayVideo(execute=@n7.then, self:Fort Activity Video Cycle Object Reference)
   n7 = FadeOutVideoBumper(execute=@n8.then) {is=title K2Node_CallFunction}
   n8 = Set bBumperFinished(execute=@n2.then, bBumperFinished=true)
  ]}
 SnapToEndOfAnimation {nodes [
   n1 = SnapToEndOfAnimation!() -> (then, InAnimation:Widget Animation Object Reference, PlayMode:EUMGSequencePlayMode Enum=Forward) {is=title K2Node_FunctionEntry}
   > PlayAnimation(InAnimation=@n1, StartAtTime=GetEndTime(self=@n1), PlayMode=@n1.PlayMode)
  ]}
]
variables [
 bIntroHasPlayed {type=boolean}
 bOutroPlaying {type=boolean}
 VideoBumperTimer {type=TimerHandle}
]
widgets {
 info {inheritedWidgetCount=1, parentClass=/Script/FortniteUI.FortActivityVideoCycle, rootWidgetClass=&Overlay, widgetCount=3}
 widgets [
  Overlay_0 {widgetClassPath=&Overlay, widget=Overlay_0}
  ActivityMovieWidget {
   widgetClassPath=&/WBP_IslandTile_VideoV2
   bInherited=true
   bIsVariable=true
   parent=Overlay_0
   widget=ActivityMovieWidget
   slot {class=OverlaySlot, properties {horizontalAlignment=HAlign_Fill, verticalAlignment=VAlign_Fill}}
  }
  SqueegeeSlot_Bumper {
   widgetClassPath=/Game/UI/Plugin/Squeegee/SqueegeeInjectionSlot
   bIsVariable=true
   parent=Overlay_0
   widget=SqueegeeSlot_Bumper
   properties {pluginId=Squeegee, pluginWidgetId=Video Bumper, renderOpacity=0}
   slot {class=OverlaySlot, properties {horizontalAlignment=HAlign_Fill, verticalAlignment=VAlign_Fill}}
  }
 ]
}
