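# Status indicator widget that shows an activity's download state as a small progress widget or a completed check icon.
# SetDownloadState switches between an empty spacer for Unknown, DownloadProgress for Downloading and Paused, and the check for ReadyToPlay.
# It shows progress feedback only while Downloading, hides it for Paused and ReadyToPlay, and resets progress on Unknown.
# SetDownloadProgress forwards the ratio, and Construct sets the progress material's fill opacity to full.
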WidgetBlueprint /FortUILibrary/UIKit/ProgressBar/WBP_StatusIndicator_DownloadProgress
graphs [
 EventGraph {nodes [
   Event Construct()
   > SetScalarParameterValue(
    self=Get DynamicProgressMaterial!(self:WBP UI Download Progress Object Reference=Get DownloadProgress() -> (DownloadProgress:WBP UI Download Progress Object Reference)) -> (DynamicProgressMaterial:Material Instance Dynamic Object Reference)
    ParameterName=Fill Opacity
    Value=1
   )
  ]}
 SetDownloadProgress {nodes [
   n1 = SetDownloadProgress!() -> (then, Ratio:double=~) {is=title K2Node_FunctionEntry}
   > SetProgress!(execute, self:WBP UI Download Progress Object Reference=Get DownloadProgress(), Value:double=@n1) -> (then)
  ]}
 SetDownloadState {nodes [
   n1 = SetDownloadState!() -> (then, State:EFortActivityViewModelDownloadState Enum=DownloadRequired) {is=title K2Node_FunctionEntry}
   n2 = Get DownloadProgress()
   > SetActiveWidget(
    self=Get Switcher_StatusIcons() -> (Switcher_StatusIcons:Common Visibility Switcher Object Reference)
    Widget=Select!(
     Unknown:Widget Object Reference=Get Unknown() -> (Unknown:Spacer Object Reference)
     Downloading:Widget Object Reference=@n2
     Paused:Widget Object Reference=@n2
     ReadyToPlay:Widget Object Reference=Get Completed() -> (Completed:WBP UIKit Status Indicator Object Reference)
     Index:EFortActivityViewModelDownloadState Enum=@n1
    ) -> (ReturnValue:Widget Object Reference)
   )
   > n3 = Switch on EFortActivityViewModelDownloadState!(execute, Selection:EFortActivityViewModelDownloadState Enum=@n1) -> (Unknown:Exec, Downloading:Exec, Paused:Exec, ReadyToPlay:Exec) {is=title K2Node_SwitchEnum}
   SetFeedbackVisibility!(execute=@n3.Downloading, self:WBP UI Download Progress Object Reference=Get DownloadProgress(), IsVisible:bool=true) -> (then)
   SetFeedbackVisibility(execute=[@n3.Paused, @n3.ReadyToPlay], self=Get DownloadProgress(), IsVisible=false)
   SetDownloadProgress!(execute=@n3.Unknown, self, Ratio:double=0) -> (then) {is=title K2Node_CallFunction}
  ]}
]
variables []
widgets {
 info {inheritedWidgetCount=0, parentClass=&UserWidget, rootWidgetClass=&Overlay, widgetCount=5}
 widgets [
  Overlay_0 {widgetClassPath=&Overlay, widget=Overlay_0}
  Switcher_StatusIcons {
   widgetClassPath=&CommonVisibilitySwitcher
   bIsVariable=true
   parent=Overlay_0
   widget=Switcher_StatusIcons
   slot {class=OverlaySlot, properties {horizontalAlignment=HAlign_Fill, verticalAlignment=VAlign_Fill}}
  }
  Unknown {
   widgetClassPath=&Spacer
   bIsVariable=true
   parent=Switcher_StatusIcons
   widget=Unknown
   properties {visibility=Collapsed, size=<0 0>}
   slot {class=CommonVisibilitySwitcherSlot}
  }
  Completed {
   widgetClassPath=/FortUILibrary/UIKit/StatusIndicator/WBP_UIKit_StatusIndicator
   bIsVariable=true
   parent=Switcher_StatusIcons
   widget=Completed
   properties {icon=/FortUILibrary/Textures/Icons/T_UI_Icon_UIKit_Check, iconSize=Compact, overrideStatusIndicatorVisibility=true}
   slot {class=CommonVisibilitySwitcherSlot, properties {horizontalAlignment=HAlign_Center}}
  }
  DownloadProgress {
   widgetClassPath=&/WBP_UI_DownloadProgress
   bIsVariable=true
   parent=Switcher_StatusIcons
   widget=DownloadProgress
   properties {widgetSize=<32 32>}
   slot {class=CommonVisibilitySwitcherSlot}
  }
 ]
}
