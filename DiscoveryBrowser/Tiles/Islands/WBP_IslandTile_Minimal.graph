WidgetBlueprint /DiscoveryBrowser/Tiles/Islands/WBP_IslandTile_Minimal
animations [
 Anim_OnLoadedImage {
  binding_count=1
  length_seconds=0.24
  object=&self:Anim_OnLoadedImage
  sequence {bindings [TileThumbnail {
     id=#1
     possessed_class=Image
     tracks [{
       type=MovieSceneWidgetMaterialTrack
       sections [{
         type=MovieSceneParameterSection
         end_frame=4
         start_frame=0
         channels [
          Thumbnail Alpha {type=MovieSceneScriptingFloatChannel, keys [[0,0],[4.8,1]]}
          Thumbnail Zoom {type=MovieSceneScriptingFloatChannel, keys [[0,1.05],[4.8,1.03]]}
         ]
        }]
      }]
    }]}
 }
 Anim_OnHoverUnhoverRehover {
  binding_count=2
  length_seconds=0.2
  object=&self:Anim_OnHoverUnhoverRehover
  sequence {bindings [
    TileThumbnail {
     id=#2
     possessed_class=Image
     tracks [{
       type=MovieSceneWidgetMaterialTrack
       sections [{type=MovieSceneParameterSection, channels [Thumbnail Zoom {type=MovieSceneScriptingFloatChannel, keys [[0,1],[4,1.1]]}]}]
      }]
    }
    NamedSlot_Title {
     id=#3
     possessed_class=NamedSlot
     tracks [{
       type=MovieSceneFloatTrack
       display_name=Render Opacity
       sections [{type=MovieSceneFloatSection, end_frame=4, start_frame=0, channels [None {type=MovieSceneScriptingFloatChannel, keys [[0,0],[4,1]]}]}]
      }]
    }
   ]}
 }
 Anim_OnFavorite {
  binding_count=1
  length_seconds=0.25
  object=&self:Anim_OnFavorite
  sequence {bindings [StatusIndicator_Favorite {
     id=#4
     possessed_class=WBP_UIKit_StatusIndicator_C
     tracks [{
       type=MovieScene2DTransformTrack
       sections [{
         type=MovieScene2DTransformSection
         end_frame=5
         start_frame=0
         channels [
          Translation.X {type=MovieSceneScriptingFloatChannel, const [0,1]}
          Translation.Y {type=MovieSceneScriptingFloatChannel, const [0,1]}
          Angle {type=MovieSceneScriptingFloatChannel, const [0,1]}
          Scale.X {type=MovieSceneScriptingFloatChannel, keys [[0,0],[3,1.2],[5,1]]}
          Scale.Y {type=MovieSceneScriptingFloatChannel, keys [[0,0],[3,1.2],[5,1]]}
          Shear.X {type=MovieSceneScriptingFloatChannel, const [0,1]}
          Shear.Y {type=MovieSceneScriptingFloatChannel, const [0,1]}
         ]
        }]
      }]
    }]}
 }
]
graphs [
 BP_OnHoldTriggered {nodes [
   n1 = On Hold Triggered!() -> (then, CurrentInputType:ECommonInputType Enum) {is=title K2Node_FunctionEntry}
   > n2 = Switch on ECommonInputType!(execute, Selection:ECommonInputType Enum=@n1) -> (MouseAndKeyboard:Exec, Gamepad:Exec, Touch:Exec, Count:Exec) {is=title K2Node_SwitchEnum}
   Branch(
    execute=@n2.Touch
    Condition=Get bEnableContextMenuInDiscover!(self:Fort Runtime Options Object Reference=GetRuntimeOptions()) -> (bEnableContextMenuInDiscover:bool)
   )
   n3 = Get WBP_Discover_Tile_Base() -> (WBP_Discover_Tile_Base:WBP Discover Tile Base Object Reference)
   > ToggleOpen!(
    execute
    self:Menu Anchor Object Reference=Get Anchor!(self:UIKit Context Menu Anchor Object Reference=Get WBP_UIKit_ContextMenuAnchorButton!(self:WBP Discover Tile Base Object Reference=@n3) -> (WBP_UIKit_ContextMenuAnchorButton:WBP UIKit Context Menu Anchor Button Object Reference)) -> (Anchor:Menu Anchor Object Reference)
    bFocusOnOpen:bool=false
   ) -> (then)
   > Return Node!(execute, ReturnValue:Event Reply Structure=Handled())
  ]}
 ConvertParamsToTileSize {nodes [
   n1 = ConvertParamsToTileSize!() -> (then, NewSize:int, IsSquareTiles:bool=~) {is=title K2Node_FunctionEntry}
   > Return Node!(
    execute
    NewTileSize:Vector 2D Structure=ColumnSizeToTileSize!(self:WBP Discover Tile Base Object Reference=Get WBP_Discover_Tile_Base(), NewColumnSize:int=@n1, IsSquareTiles:bool=@n1.IsSquareTiles) -> (NewTileSize:Vector 2D Structure)
   )
  ]}
 EventGraph {nodes [
   AddInfoModalToContextMenu() {is=title K2Node_CustomEvent}
   n1 = Get FortActivityViewModel() -> (FortActivityViewModel:Discover Activity View Model Object Reference)
   > n2 = Set IsRecommendedModal(IsRecommendedModal:bool="Equal (Enum)"(
     A:EFortCreativeDiscoveryPanelType Enum=Get PanelType!(self:Discover Activity View Model Object Reference=@n1) -> (PanelType:EFortCreativeDiscoveryPanelType Enum=CuratedList)
     B:EFortCreativeDiscoveryPanelType Enum=Recommendations
    )) -> (Output_Get:bool)
   n3 = OpenInfoModal!() -> (OutputDelegate, then, InteractionNameID:Name) {is=title K2Node_CustomEvent}
   AddInteraction!(
    execute=@n2.then
    self:UIKit Context Menu Action Context Object Reference=Get ContextMenuActionContext!(self:WBP Discover Tile Base Object Reference=Get WBP_Discover_Tile_Base()) -> (ContextMenuActionContext:UIKit Context Menu Action Context Object Reference)
    InActionInfo_DisplayName:Text="LOCTEXT(\"Why You\\'re Seeing This\")"
    InActionInfo_IconTexture:Texture 2D Soft Object Reference
    InActionInfo_IconMaterial:Material Instance Soft Object Reference
    InActionInfo_SortPriority:int=1001
    InActionInfo_bIsAvailable:bool=True
    InActionInfo_bIsEnabled:bool=True
    InActionInfo_bIsSelected:bool=False
    InActionInfo_bAddDivider:bool=False
    InActionInfo_bPositionDividerAfter:bool=False
    InActionInfo_bCloseMenuOnSelect:bool=False
    InActionInfo_InteractionNameID:Name=Get RecommendModalName() -> (RecommendModalName:Name=None)
    InActionInfo_Interaction:Data Table Row Handle Structure=(DataTable=None,RowName="")
    InActionInfo_PressedDelegate:Delegate="(null).None"
    InActionInfo_ReleasedDelegate:Delegate="(null).None"
    InInteractionDelegate:"Delegate (by ref)"=@n3
   ) -> (then)
   n4 = OnModalDismissed!() -> (OutputDelegate, then, ButtonIndex:int) {is=title K2Node_CustomEvent}
   n5 = Construct Object from Class(execute=@n77.else, Class=/DiscoveryBrowser/Dialogs/DialogVM_DiscoTileDisclaimer_RecentlyPlayed.DialogVM_DiscoTileDisclaimer_RecentlyPlayed_C) -> (ReturnValue:Dialog VM Disco Tile Disclaimer Recently Played Object Reference)
   n6 = Construct Object from Class(execute=@n77.then, Class=/DiscoveryBrowser/Dialogs/DialogVM_DiscoTileDisclaimer_Recommended.DialogVM_DiscoTileDisclaimer_Recommended_C) -> (ReturnValue:Dialog VM Disco Tile Disclaimer Recommended Object Reference)
   n7 = Create Widget!(
    execute:Exec=[@n5.then, @n6.then]
    Class:User Widget Class Reference=/FortUILibrary/UIKit/Dialogs/WBP_UIKit_Dialog.WBP_UIKit_Dialog_C
    OwningPlayer:Player Controller Object Reference
    UIKitDialogViewModel:UI Kit Dialog Viewmodel Object Reference=Select(
     Option 0:UI Kit Dialog Viewmodel Object Reference=@n5
     Option 1:UI Kit Dialog Viewmodel Object Reference=@n6
     Index=Get IsRecommendedModal() -> (IsRecommendedModal:bool)
    ) -> (ReturnValue:UI Kit Dialog Viewmodel Object Reference)
   ) -> (then:Exec, ReturnValue:WBP UIKit Dialog Object Reference)
   n8 = Get Dialog_Internal!(self:WBP UIKit Dialog Object Reference=@n7) -> (Dialog_Internal:WBP UIKit Dialog Internal Object Reference)
   Unbind all Events from Dialog Button Clicked Dispatcher!(execute:Exec=@n4.then, self:WBP UIKit Dialog Internal Object Reference=@n8) -> (then:Exec) {is=title K2Node_ClearDelegate}
   > DeactivateWidget(self=@n7)
   n9 = Set Up Favorite Interaction!() -> (OutputDelegate, then, IsFavorited:bool=~) {is=title K2Node_CustomEvent}
   > n10 = RemoveInteractionByName!(
    execute
    self:WBP Discover Tile Base Object Reference=Get WBP_Discover_Tile_Base()
    ActionToRemove:Name=Get ContextMenuFavoriteName() -> (ContextMenuFavoriteName:Name=None)
   ) -> (then)
   n11 = ToggleFavorite!() -> (OutputDelegate, then, InteractionNameID:Name) {is=title K2Node_CustomEvent}
   AddInteraction(
    execute=@n10.then
    self=Get ContextMenuActionContext(self=Get WBP_Discover_Tile_Base())
    InActionInfo_DisplayName=SelectText(A="LOCTEXT(\"Remove from Saved\")", B="LOCTEXT(\"Save the Island\")", bPickA=@n9.IsFavorited)
    InActionInfo_SortPriority=999
    InActionInfo_bIsAvailable=True
    InActionInfo_bIsEnabled=True
    InActionInfo_bIsSelected=False
    InActionInfo_bAddDivider=False
    InActionInfo_bPositionDividerAfter=False
    InActionInfo_bCloseMenuOnSelect=False
    InActionInfo_InteractionNameID=Get ContextMenuFavoriteName() -> (ContextMenuFavoriteName=None)
    InActionInfo_Interaction=(DataTable=None,RowName="")
    InActionInfo_PressedDelegate="(null).None"
    InActionInfo_ReleasedDelegate="(null).None"
    InInteractionDelegate=@n11
   )
   n12 = ToggleNotInterested!() -> (OutputDelegate, then, InteractionNameID:Name) {is=title K2Node_CustomEvent}
   > MarkAsNotInterested(self:Fort Activity Selector Object Reference=GetActivitySelector!(self:Fort Activity Tile Details Display Object Reference) -> (ReturnValue:Fort Activity Selector Object Reference))
   > UpdateContextMenuNotInterested!(execute, self, ShouldAddNotInterested:bool=false) -> (then)
   > CloseContextMenu(self:WBP Discover Tile Base Object Reference=Get WBP_Discover_Tile_Base())
   n13 = MoreFromCreator!() -> (OutputDelegate, then, InteractionNameID:Name) {is=title K2Node_CustomEvent}
   > OpenCreatorPage!(execute, self:Fort Activity Selector Object Reference=GetActivitySelector(), bIsOpenedFromDetailsModal:bool=false) -> (then)
   > CloseContextMenu(self=Get WBP_Discover_Tile_Base())
   n14 = ViewDetails!() -> (OutputDelegate, then, InteractionNameID:Name) {is=title K2Node_CustomEvent}
   > NotifyAnalyticsTileWasClicked(self:Fort Activity Tile Details Display Object Reference)
   n15 = GetActivitySelector()
   > SetSelectedActivityShouldShowLikePercent!(
    execute
    self:Fort Activity Selector Object Reference=@n15
    bInShouldShowLikePercent:bool=ShouldDisplayLikePercent!(self:Discover Activity View Model Object Reference=Get FortActivityViewModel()) -> (ReturnValue:bool)
   ) -> (then)
   > HandleShowActivityDetails(self:Fort Activity Selector Object Reference=@n15)
   > CloseContextMenu(self=Get WBP_Discover_Tile_Base())
   n16 = GoToLobby!() -> (OutputDelegate, then, InteractionNameID:Name) {is=title K2Node_CustomEvent}
   > NotifyAnalyticsTileWasClicked()
   > ConfirmSelectedActivity(self:Fort Activity Selector Object Reference=GetActivitySelector())
   > CloseContextMenu(self=Get WBP_Discover_Tile_Base())
   n17 = OnContextMenuVisibilityChanged!() -> (OutputDelegate, then, IsVisible:bool=~) {is=title K2Node_CustomEvent}
   > SetVisibility(
    self=Get SqueegeeSlot_Lock() -> (SqueegeeSlot_Lock:Squeegee Injection Slot Object Reference)
    InVisibility=Select(Option 0:ESlateVisibility Enum=HitTestInvisible, Option 1:ESlateVisibility Enum=Collapsed, Index=@n17.IsVisible) -> (ReturnValue:ESlateVisibility Enum)
   )
   > Set!(
    execute
    Padding_Left:float=0
    Padding_Top:float=0
    Padding_Right:float=SelectFloat(A=48, B=0, bPickA=@n17.IsVisible)
    Padding_Bottom:float=0
    self:Horizontal Box Slot Object Reference=SlotAsHorizontalBoxSlot!(Widget:Widget Object Reference=Get VerticalBox_TitleAndCCU() -> (VerticalBox_TitleAndCCU:Vertical Box Object Reference)) -> (ReturnValue:Horizontal Box Slot Object Reference)
   ) -> (then, Output_Get:Margin Structure)
   n18 = FullTileContextMenuSetup!() -> (OutputDelegate, then, Widget:User Widget Object Reference) {is=title K2Node_CustomEvent}
   > SetInteractionAvailability!(
    execute
    self
    InteractionID:Name=Get ContextMenuGoToLobbyName() -> (ContextMenuGoToLobbyName:Name=None)
    Availability:bool=Is Valid Activity!() -> (IsValid:bool=~) {is=title K2Node_MacroInstance}
   ) -> (then)
   n19 = OnModalClosed!() -> (OutputDelegate, then, ButtonIndex:int) {is=title K2Node_CustomEvent}
   n20 = Construct Object from Class(execute=@n78.then, Class=/DiscoveryBrowser/Dialogs/DialogVM_DiscoTileDisclaimer.DialogVM_DiscoTileDisclaimer_C) -> (ReturnValue:Dialog VM Disco Tile Disclaimer Object Reference)
   n21 = Create Widget(execute=@n80.then, Class=&WBP_UIKit_Dialog_C, UIKitDialogViewModel=@n20)
   n22 = Get Dialog_Internal(self=@n21)
   Unbind all Events from Dialog Button Clicked Dispatcher(execute=@n19.then, self=@n22)
   > DeactivateWidget(self=@n21)
   AddContextMenuItems() {is=title K2Node_CustomEvent}
   > n23 = Sequence!(execute) -> (then_0, then_1, then_2:Exec)
   > RemoveInteractionByName(self=Get WBP_Discover_Tile_Base(), ActionToRemove=Get ContextMenuViewCreatorName() -> (ContextMenuViewCreatorName:Name=None))
   > n24 = Branch(Condition=Get IsPlayTimeOut() -> (IsPlayTimeOut:bool))
   n25 = Branch(execute=@n24.else, Condition=Get IsTileInCreatorPage() -> (IsTileInCreatorPage:bool))
   > Reroute Node()
   AddInteraction(
    execute=@n25.else
    self=Get ContextMenuActionContext(self=Get WBP_Discover_Tile_Base())
    InActionInfo_DisplayName="LOCTEXT(\"View Developer\")"
    InActionInfo_SortPriority=1000
    InActionInfo_bIsAvailable=True
    InActionInfo_bIsEnabled=True
    InActionInfo_bIsSelected=False
    InActionInfo_bAddDivider=False
    InActionInfo_bPositionDividerAfter=False
    InActionInfo_bCloseMenuOnSelect=False
    InActionInfo_InteractionNameID=Get ContextMenuViewCreatorName() -> (ContextMenuViewCreatorName=None)
    InActionInfo_Interaction=(DataTable=None,RowName="")
    InActionInfo_PressedDelegate="(null).None"
    InActionInfo_ReleasedDelegate="(null).None"
    InInteractionDelegate=@n13
   )
   n26 = Branch(execute=@n23.then_1, Condition=Get IsFullTile() -> (IsFullTile:bool))
   > RemoveInteractionByName(self=Get WBP_Discover_Tile_Base(), ActionToRemove=Get ContextMenuGoToLobbyName() -> (ContextMenuGoToLobbyName=None))
   > n27 = Branch(Condition=Get IsPlayTimeOut())
   n28 = Get FortGlobalUIContext() -> (ReturnValue:Fort Global UIContext Object Reference) {is=title K2Node_GetSubsystem}
   Branch(execute=@n27.else, Condition=NOT Boolean(A=IsInZone!(self:Fort Global UIContext Object Reference=@n28) -> (ReturnValue:bool)))
   > AddInteraction(
    self=Get ContextMenuActionContext(self=Get WBP_Discover_Tile_Base())
    InActionInfo_DisplayName="LOCTEXT(\"Go to Lobby\")"
    InActionInfo_SortPriority=10
    InActionInfo_bIsAvailable=True
    InActionInfo_bIsEnabled=True
    InActionInfo_bIsSelected=False
    InActionInfo_bAddDivider=False
    InActionInfo_bPositionDividerAfter=False
    InActionInfo_bCloseMenuOnSelect=False
    InActionInfo_InteractionNameID=Get ContextMenuGoToLobbyName() -> (ContextMenuGoToLobbyName=None)
    InActionInfo_Interaction=(DataTable=None,RowName="")
    InActionInfo_PressedDelegate="(null).None"
    InActionInfo_ReleasedDelegate="(null).None"
    InInteractionDelegate=@n16
   )
   RemoveInteractionByName(execute=@n26.else, self=Get WBP_Discover_Tile_Base(), ActionToRemove=Get ContextMenuViewDetails() -> (ContextMenuViewDetails:Name=None))
   > n29 = Branch(Condition=Get IsPlayTimeOut())
   AddInteraction(
    execute=@n29.else
    self=Get ContextMenuActionContext(self=Get WBP_Discover_Tile_Base())
    InActionInfo_DisplayName="LOCTEXT(\"View Details\")"
    InActionInfo_SortPriority=10
    InActionInfo_bIsAvailable=True
    InActionInfo_bIsEnabled=True
    InActionInfo_bIsSelected=False
    InActionInfo_bAddDivider=False
    InActionInfo_bPositionDividerAfter=False
    InActionInfo_bCloseMenuOnSelect=False
    InActionInfo_InteractionNameID=Get ContextMenuViewDetails() -> (ContextMenuViewDetails=None)
    InActionInfo_Interaction=(DataTable=None,RowName="")
    InActionInfo_PressedDelegate="(null).None"
    InActionInfo_ReleasedDelegate="(null).None"
    InInteractionDelegate=@n14
   )
   Branch(execute=@n23.then_2, Condition=Get IsFullTile())
   > RemoveInteractionByName(self=Get WBP_Discover_Tile_Base(), ActionToRemove=Get ContextMenuGoToLobbyName() -> (ContextMenuGoToLobbyName=None))
   > n30 = Branch(Condition=Get IsPlayTimeOut())
   Bind Event to On Pre Spawn Menu Delegate(
    execute=@n30.else
    self:UIKit Context Menu Anchor Object Reference=Get WBP_UIKit_ContextMenuAnchorButton(self=Get WBP_Discover_Tile_Base())
    Delegate=@n18
   ) {is=title K2Node_AddDelegate}
   SetUpNotInterestedInteraction() {is=title K2Node_CustomEvent}
   > RemoveInteractionByName(self=Get WBP_Discover_Tile_Base(), ActionToRemove=Get ContextMenuNotInterestedName() -> (ContextMenuNotInterestedName:Name=None))
   > n31 = Branch(Condition=Get IsPlayTimeOut())
   AddInteraction(
    execute=@n31.else
    self=Get ContextMenuActionContext(self=Get WBP_Discover_Tile_Base())
    InActionInfo_DisplayName="LOCTEXT(\"Not Interested\")"
    InActionInfo_SortPriority=1000
    InActionInfo_bIsAvailable=True
    InActionInfo_bIsEnabled=True
    InActionInfo_bIsSelected=False
    InActionInfo_bAddDivider=False
    InActionInfo_bPositionDividerAfter=False
    InActionInfo_bCloseMenuOnSelect=False
    InActionInfo_InteractionNameID=Get ContextMenuNotInterestedName() -> (ContextMenuNotInterestedName=None)
    InActionInfo_Interaction=(DataTable=None,RowName="")
    InActionInfo_PressedDelegate="(null).None"
    InActionInfo_ReleasedDelegate="(null).None"
    InInteractionDelegate=@n12
   )
   n32 = WrapperCleanupContextMenuOnRelease() {is=title K2Node_CustomEvent}
   CleanupContextMenuOnRelease(execute=[@n46.then, @n32.then], self:WBP Discover Tile Base Object Reference=Get WBP_Discover_Tile_Base())
   n33 = Event PreConstruct()
   > n34 = Sequence!(execute) -> (then_0, then_1, then_2, then_3:Exec)
   > Branch(Condition=@n33.IsDesignTime)
   n35 = Get Slot_OutsideContent() -> (Slot_OutsideContent:Named Slot Object Reference)
   > SetVisibility(
    self=@n35
    InVisibility=Select(Option 0=Collapsed, Option 1=SelfHitTestInvisible, Index=HasAnyChildren!(self:Panel Widget Object Reference=@n35) -> (ReturnValue:bool))
   )
   SetSlotOutsideAlignment!(
    execute=@n34.then_1
    self
    SlotOutsideAlignment:E_UI_Alignment Enum=Get SlotOutsideAlignment() -> (SlotOutsideAlignment:E_UI_Alignment Enum=NewEnumerator0)
   ) -> (then)
   > SetVisibility(
    self=[
     Get Badge_CCUCount() -> (Badge_CCUCount:WBP UIKit Badge Object Reference)
     Get Badge_LikePercent() -> (Badge_LikePercent:WBP UIKit Badge Object Reference)
    ]
    InVisibility=Select(Option 0=HitTestInvisible, Option 1=Collapsed, Index=Get DisableCornerInfo() -> (DisableCornerInfo:bool))
   )
   > n36 = Branch(Condition=Get HandlesOwnStartVisibilityImpression() -> (HandlesOwnStartVisibilityImpression:bool))
   > n37 = StartVisibilityImpression(self:Discover Tile Item View Model Object Reference=Get FortActivityViewModel())
   Set Icon Mobile Multiplier!(execute=[@n37.then, @n36.else], self, ImageObject:Image Object Reference=Get CMS_CreatorLogo() -> (CMS_CreatorLogo:Image Object Reference)) -> (then) {is=title K2Node_CallFunction}
   > Set Icon Mobile Multiplier(ImageObject=Get ByEpicLogo() -> (ByEpicLogo:Image Object Reference))
   > SetMobileLayout()
   > Set Justification!(
    execute
    Justification:ETextJustify Enum=Left
    self:Text Layout Widget Object Reference=Get Text_Highlight!(self:WBP UIKit Highlight Base Object Reference=Get Highlight_Label() -> (Highlight_Label:WBP UIKit Highlight Object Reference)) -> (Text_Highlight:Common Text Object Reference)
   ) -> (then, Output_Get:ETextJustify Enum=Left)
   SetVisibility(execute=@n34.then_2, self=Get Badge_Countdown() -> (Badge_Countdown:WBP UIKit Badge Object Reference), InVisibility=Collapsed)
   n38 = Event OnTileActiveSet!() -> (OutputDelegate, then, bIsTileActive:bool) {is=title K2Node_Event}
   > Branch(Condition=Get IsOldDiscoverCode() -> (IsOldDiscoverCode:bool))
   > n39 = Branch(Condition=@n38.bIsTileActive)
   HandleTileSelectionActive!(execute=[@n65.then, @n54.then, @n39.then, @n45.then], self, InActive:bool=true) -> (then)
   HandleTileSelectionActive(execute=[@n64.then, @n39.else, @n48.then, @n66.then], InActive=false)
   Event Construct()
   > AddContextMenuItems() {is=title K2Node_CallFunction}
   n40 = Get Text_Highlight(self=Get Highlight_Label())
   > SetScrollingEnabled!(execute, self:Common Text Object Reference=@n40, bInIsScrollingEnabled:bool=false) -> (then)
   > Set TextOverflowPolicy!(execute, TextOverflowPolicy:ETextOverflowPolicy Enum=Ellipsis, self:Text Object Reference=@n40) -> (then, Output_Get:ETextOverflowPolicy Enum=Clip)
   > Bind Event to On Context Menu Visibility Changed(self:WBP Discover Tile Base Object Reference=Get WBP_Discover_Tile_Base(), Delegate=Create Event()) {is=title K2Node_AddDelegate}
   > InitializeClickedSound()
   > SetDownloadProgressVisibility!(execute, self, IsVisible:bool=false) -> (then)
   Event Destruct()
   n41 = Get WBP_Discover_Tile_Base()
   > Unbind all Events from On Context Menu Visibility Changed!(execute:Exec, self:WBP Discover Tile Base Object Reference=@n41) -> (then:Exec) {is=title K2Node_ClearDelegate}
   > Unbind all Events from On Set Primary Text from VM!(execute:Exec, self:WBP Discover Tile Base Object Reference=@n41) -> (then:Exec) {is=title K2Node_ClearDelegate}
   n42 = Event On Input Method Changed!() -> (OutputDelegate, then, CurrentInputType:ECommonInputType Enum) {is=title K2Node_Event}
   > n43 = Branch(Condition=Get IsOldDiscoverCode())
   SetIsSelectable!(
    execute=@n43.else
    self:Common Button Base Object Reference
    bInIsSelectable:bool=Select!(MouseAndKeyboard:bool=false, Gamepad:bool=false, Touch:bool=true, Count:bool=false, Index:ECommonInputType Enum=@n42.CurrentInputType) -> (ReturnValue:bool)
   ) -> (then)
   Event On Focused() {is=title K2Node_Event}
   > Call On Focus Gained!(execute:Exec, self:WBP Island Tile Minimal Object Reference, FocusedItem:WBP Island Tile Minimal Object Reference=Self-Reference()) -> (then:Exec) {is=title K2Node_CallDelegate}
   n44 = IsUsingFocusNavigationOnKBM!() -> (IsUsingFocusNavigationForOutlineVisuals:bool)
   > n45 = Branch(Condition=@n44)
   Event On Entry Released() {is=title K2Node_Event}
   > n46 = ClearDebugUI(self:WBP Discover Tile Base Object Reference=Get WBP_Discover_Tile_Base())
   Event On Unfocused() {is=title K2Node_Event}
   n47 = IsHovered()
   > Branch(Condition=AND Boolean(A=@n44, B=NOT Boolean(A=@n47)))
   > n48 = Branch(Condition=Get IsTileActive!(self:WBP Discover Tile Base Object Reference=Get WBP_Discover_Tile_Base()) -> (IsTileActive:bool))
   Event OnDetailsUpdated() {is=title K2Node_Event}
   > FortActivityViewModel_OnDownloadStateChanged!(
    execute
    self
    DownloadState:EFortActivityViewModelDownloadState Enum=Get DownloadState!(self:Discover Activity View Model Object Reference=Get FortActivityViewModel()) -> (DownloadState:EFortActivityViewModelDownloadState Enum=Unknown)
   ) -> (then)
   > ConfigureDebugUI!(
    execute
    self:WBP Discover Tile Base Object Reference=Get WBP_Discover_Tile_Base()
    Debug VM:Discover Tile Item Debug View Model Object Reference=Get DebugVM!(self:Discover Tile Item View Model Object Reference=Get FortActivityViewModel()) -> (DebugVM:Discover Tile Item Debug View Model Object Reference)
    OwningButton:Common Button Base Object Reference=Self-Reference()
   ) -> (then)
   > SetupCountdownTimerViewModel!(execute, self, DiscoverActivityViewModel:Discover Activity View Model Object Reference=Get FortActivityViewModel()) -> (then)
   n49 = Event SetNestedButtonGroupSelectionVM!() -> (OutputDelegate, then, InVM:Fort Nested Button Group Selection VM Object Reference) {is=title K2Node_Event}
   > Set FortNestedButtonGroupSelectionVM(FortNestedButtonGroupSelectionVM:Fort Nested Button Group Selection VM Object Reference=@n49.InVM) -> (Output_Get:Fort Nested Button Group Selection VM Object Reference)
   Event OnViewportResizedEvent() {is=title K2Node_Event}
   n50 = Get FortActivityViewModel()
   > Set Tile Size!(
    execute
    self:WBP Discover Tile Base Object Reference=Get WBP_Discover_Tile_Base()
    InSize:Vector 2D Structure=ConvertParamsToTileSize!(
     self
     NewSize:int=Get ColumnSize!(self:Discover Tile Item View Model Object Reference=@n50) -> (ColumnSize:int=0)
     IsSquareTiles:bool=Get bDisplayAsSquareTile!(self:Discover Tile Item View Model Object Reference=@n50) -> (bDisplayAsSquareTile:bool)
    ) -> (NewTileSize:Vector 2D Structure) {is=title K2Node_CallFunction}
   ) -> (then) {is=title K2Node_CallFunction}
   Event OnInitialized()
   > Override Right Click Sound!(
    execute
    self:WBP Discover Tile Base Object Reference=Get WBP_Discover_Tile_Base()
    New Right Click Sound:Sound Base Object Reference=Get Right Clicked Sound Override() -> (Right Clicked Sound Override:Sound Base Object Reference)
   ) -> (then)
   Event On Hovered()
   > n51 = Branch(Condition=Get bEnableVideosInDiscover!(self:Fort Runtime Options Object Reference=GetRuntimeOptions()) -> (bEnableVideosInDiscover:bool))
   > SetVisibility(self=Get ActivityVideoCycleWidget() -> (ActivityVideoCycleWidget:Fort Activity Video Cycle Object Reference), InVisibility=HitTestInvisible)
   > n52 = StartTileVideo(self:Fort Activity Tile Details Display Object Reference)
   n53 = NOR Boolean!(A:bool=Get IsOldDiscoverCode(), B:bool=Get bSelectable() -> (bSelectable:bool)) -> (ReturnValue:bool)
   Branch(execute=[@n52.then, @n51.else], Condition=@n53)
   > n54 = Parent: On Hovered() {is=title K2Node_CallParentFunction}
   Event On Pressed() {is=title K2Node_Event}
   > Branch(Condition=Get AllowTileClicks() -> (AllowTileClicks:bool))
   > PlayPressedSound()
   n55 = Event OnListItemObjectSet()
   > n56 = Cast To FortActivityViewModel!(execute, Object=@n55.ListItemObject) -> (then, CastFailed, AsDiscover Activity View Model:Discover Activity View Model Object Reference) {is=title K2Node_DynamicCast}
   > n57 = Set FortActivityViewModel(FortActivityViewModel:Discover Activity View Model Object Reference=@n56) -> (Output_Get:Discover Activity View Model Object Reference)
   > UpdateFromFortActivityViewModel(self:Fort Activity Tile Details Display Object Reference)
   > SetupCountdownTimerViewModel(DiscoverActivityViewModel=@n57)
   > n58 = Sequence!(execute) -> (then_0, then_1)
   > n59 = Branch(Condition=HasFetchedImages!(self:Fort Discover Link Entry Tile Item VM Object Reference=@n57) -> (ReturnValue:bool))
   FetchImages(execute=@n59.else, self:Fort Discover Link Entry Tile Item VM Object Reference=@n57)
   Branch(execute=@n58.then_1, Condition=Get HandlesOwnStartVisibilityImpression())
   > StartVisibilityImpression(self=@n57)
   Event On Clicked()
   > Branch(Condition=Get AllowTileClicks())
   n60 = Get FortActivityViewModel()
   > n61 = LinkIdToString!(
    execute
    LinkString:"Online Link Id Structure (by ref)"=Get LinkId!(self:Fort Game Activity Object Reference=GetGameActivity!(self:Discover Activity View Model Object Reference=@n60) -> (ReturnValue:Fort Game Activity Object Reference)) -> (LinkId:Online Link Id Structure)
   ) -> (then, ReturnValue:String)
   > SendFlowRiderClickTileEvent!(
    execute
    self:WBP Discover Tile Base Object Reference=Get WBP_Discover_Tile_Base()
    TileWidget:User Widget Object Reference=Self-Reference()
    TileType:EActivityItemType Enum=Island
    TileUniqueIdentifier:String=@n61
   ) -> (then)
   > OnTileClicked(self:Fort Activity Tile Details Display Object Reference)
   n62 = GetOwningPlayer()
   > n63 = Branch(Condition=AND Boolean(A=IsUsingTouch(PlayerController=@n62), B=Get bSelectable()))
   NotifyAnalyticsTileWasClicked(execute=@n63.else)
   Event On Unhovered()
   > StopTileVideo(self:Fort Activity Tile Details Display Object Reference)
   > n64 = Branch(Condition=@n53)
   Event On Selected()
   > Branch(Condition=Get bSelectable())
   > Parent: On Selected() {is=title K2Node_CallParentFunction}
   > n65 = NotifyAnalyticsTileWasClicked()
   Event On Deselected()
   > Branch(Condition=Get bSelectable())
   > n66 = Parent: On Deselected() {is=title K2Node_CallParentFunction}
   n67 = Event OnRespondToTileViewVisibilityChange!() -> (OutputDelegate, then, bIsVisible:bool) {is=title K2Node_Event}
   > n68 = Branch(Condition=@n67.bIsVisible)
   > Branch(Condition=Get HandlesOwnStartVisibilityImpression())
   n69 = Get FortActivityViewModel()
   > StartVisibilityImpression(self=@n69)
   StopVisibilityImpression(execute=@n68.else, self:Discover Tile Item View Model Object Reference=@n69)
   Event On Released() {is=title K2Node_Event}
   > StopVisibilityImpression(self=Get FortActivityViewModel())
   > StopTileActiveImpressions(self:WBP Discover Tile Base Object Reference=Get WBP_Discover_Tile_Base())
   CloseContextMenu(execute=[@n71.then, @n70.then], self=Get WBP_Discover_Tile_Base())
   CloseContextMenu(execute=@n74.then, self=Get WBP_Discover_Tile_Base())
   Bind Event to Dialog Button Clicked Dispatcher(execute=@n7.then, self:WBP UIKit Dialog Internal Object Reference=@n8, Delegate=@n4) {is=title K2Node_AddDelegate}
   > n70 = TryPushWidgetAsModalPanel!(execute, PushWidget:User Widget Object Reference=@n7) -> (then)
   Bind Event to Dialog Button Clicked Dispatcher(execute=@n21.then, self=@n22, Delegate=@n19)
   > n71 = TryPushWidgetAsModalPanel(PushWidget=@n21)
   n72 = GetRuntimeOptions()
   n73 = PlaySound2D(execute=@n79.then, Sound=Get Unfavorited Sound() -> (Unfavorited Sound:Sound Base Object Reference))
   n74 = ToggleFavorite(execute=[@n73.then, @n75.then], self:Fort Activity Selector Object Reference=GetActivitySelector()) {is=title K2Node_CallFunction}
   n75 = PlaySound2D(execute=@n79.else, Sound=Get Favorited Sound() -> (Favorited Sound:Sound Base Object Reference))
   n76 = Find!(
    TargetMap:Map of Strings to Reason Surfaced Info Structures=Get ReasonSurfacedInfo!(self:Fort Runtime Options Object Reference=@n72) -> (ReasonSurfacedInfo:Map of Strings to Reason Surfaced Info Structures)
    Key=Get ReasonSurfaced!(self:Discover Activity View Model Object Reference=@n1) -> (ReasonSurfaced:String)
   ) -> (Value_Title:Text, Value_Info:Text, ReturnValue)
   n77 = Branch(execute=@n78.else, Condition=Get IsRecommendedModal())
   n78 = Branch(execute=@n3.then, Condition=@n76.ReturnValue)
   n79 = Branch(execute=@n11.then, Condition=@n9.IsFavorited)
   Set DialogTitle!(execute=@n20.then, DialogTitle:Text=@n76, self:UI Kit Dialog Viewmodel Object Reference=@n20) -> (then, Output_Get:Text)
   > n80 = Set DialogInfoText!(execute, DialogInfoText:Text=@n76.Value_Info, self:UI Kit Dialog Viewmodel Object Reference=@n20) -> (then, Output_Get:Text)
  ]}
 FortActivityViewModel_CancelDownload {nodes [
   FortActivityViewModel_CancelDownload!() -> (then, InteractionNameID:Name) {is=title K2Node_FunctionEntry}
   > CancelDownload(self:Discover Activity View Model Object Reference=Get FortActivityViewModel())
   > CloseContextMenu(self=Get WBP_Discover_Tile_Base())
  ]}
 FortActivityViewModel_DownloadFirst {nodes [
   FortActivityViewModel_DownloadFirst!() -> (then, InteractionNameID:Name) {is=title K2Node_FunctionEntry}
   > DownloadFirst(self:Discover Activity View Model Object Reference=Get FortActivityViewModel())
   > CloseContextMenu(self=Get WBP_Discover_Tile_Base())
  ]}
 FortActivityViewModel_OnDownloadStateChanged {nodes [
   n1 = FortActivityViewModel_OnDownloadStateChanged!() -> (then, DownloadState:EFortActivityViewModelDownloadState Enum=Unknown) {is=title K2Node_FunctionEntry}
   > n2 = Branch(Condition=Get IsFullTile())
   n3 = Get FortActivityViewModel()
   n4 = GetRuntimeOptions()
   Branch(
    execute=@n2.else
    Condition=AND Boolean(
     A=Get bEnableThinClientQueuedDownloads!(self:Fort Runtime Options Object Reference=@n4) -> (bEnableThinClientQueuedDownloads:bool)
     B=Get bIsDownloadable!(self:Discover Activity View Model Object Reference=@n3) -> (bIsDownloadable:bool)
    )
   )
   > n5 = Sequence()
   n6 = Get ContextMenuActionContext(self=Get WBP_Discover_Tile_Base())
   > RemoveInteraction!(
    execute
    self:UIKit Context Menu Action Context Object Reference=@n6
    InteractionNameID:Name=Get ContextMenuDownloadFirstName() -> (ContextMenuDownloadFirstName:Name=None)
   ) -> (then)
   > RemoveInteraction(self=@n6, InteractionNameID=Get ContextMenuCancelDownloadName() -> (ContextMenuCancelDownloadName:Name=None))
   n7 = Switch on EFortActivityViewModelDownloadState!(execute=@n5.then_1, Selection:EFortActivityViewModelDownloadState Enum=@n1) -> (Unknown:Exec, Downloading:Exec, Paused:Exec, ReadyToPlay:Exec) {is=title K2Node_SwitchEnum}
   AddInteraction(
    execute=[@n7.Downloading, @n8.then]
    self=Get ContextMenuActionContext(self=Get WBP_Discover_Tile_Base())
    InActionInfo_DisplayName="LOCTEXT(\"Cancel Download\")"
    InActionInfo_SortPriority=6
    InActionInfo_bIsAvailable=True
    InActionInfo_bIsEnabled=True
    InActionInfo_bIsSelected=False
    InActionInfo_bAddDivider=true
    InActionInfo_bPositionDividerAfter=true
    InActionInfo_bCloseMenuOnSelect=False
    InActionInfo_InteractionNameID=Get ContextMenuCancelDownloadName() -> (ContextMenuCancelDownloadName=None)
    InActionInfo_Interaction=(DataTable=None,RowName="")
    InActionInfo_PressedDelegate="(null).None"
    InActionInfo_ReleasedDelegate="(null).None"
    InInteractionDelegate=Create Event()
   )
   n8 = AddInteraction(
    execute=@n7.Paused
    self=Get ContextMenuActionContext(self=Get WBP_Discover_Tile_Base())
    InActionInfo_DisplayName="LOCTEXT(\"Download First\")"
    InActionInfo_SortPriority=5
    InActionInfo_bIsAvailable=True
    InActionInfo_bIsEnabled=True
    InActionInfo_bIsSelected=False
    InActionInfo_bAddDivider=False
    InActionInfo_bPositionDividerAfter=False
    InActionInfo_bCloseMenuOnSelect=False
    InActionInfo_InteractionNameID=Get ContextMenuDownloadFirstName() -> (ContextMenuDownloadFirstName=None)
    InActionInfo_Interaction=(DataTable=None,RowName="")
    InActionInfo_PressedDelegate="(null).None"
    InActionInfo_ReleasedDelegate="(null).None"
    InInteractionDelegate=Create Event()
   )
  ]}
 Get Should Add Info Modal to Context Menu {nodes [
   n1 = Get Should Add Info Modal to Context Menu!() -> (then, PanelType:EFortCreativeDiscoveryPanelType Enum=CuratedList, ReasonSurfaced:String, PlaytimeState:EPlaytimeState Enum=Unblocked) {is=title K2Node_FunctionEntry}
   n2 = "Equal (Enum)"(A:EPlaytimeState Enum=Get PlaytimeState() -> (PlaytimeState:EPlaytimeState Enum=Unblocked), B:EPlaytimeState Enum=PlayTime_Blocked)
   n3 = AND Boolean(
    A="Equal (Enum)"(A:EFortCreativeDiscoveryPanelType Enum=@n1, B:EFortCreativeDiscoveryPanelType Enum=PlayHistory)
    B=NOT Boolean(A=Get IsFullTile())
   )
   n4 = "Equal (Enum)"(A=@n1, B=Recommendations)
   n5 = NOT Boolean(A=IsEmpty(InString=@n1.ReasonSurfaced))
   > Return Node!(execute, ShouldAddInfoModal:bool=AND Boolean(A=OR Boolean!(A=@n4, B=@n3, C:bool=@n5) -> (ReturnValue), B=NOT Boolean(A=@n2)))
  ]}
 GetCCUCountVisibility {nodes [
   GetCCUCountVisibility!() -> (
    then
    CCUCount:int
    DownloadState:EFortActivityViewModelDownloadState Enum=Unknown
    IsContentGated:bool=~
    TimerState:ECountdownTimerState Enum=None
    HasHideCCUFeatureTag:bool=~
    HasShowLikePercentFeatureTag:bool=~
    ShouldShowCCUAndLikePercent:bool=~
   ) {is=title K2Node_FunctionEntry}
   n1 = NOT Boolean(A=Get HasHideCCUFeatureTag() -> (HasHideCCUFeatureTag:bool))
   n2 = Get TimerState() -> (TimerState:ECountdownTimerState Enum=None)
   n3 = Get DownloadState!() -> (DownloadState=Unknown)
   n4 = NOT Boolean(A=OR Boolean(
     A="Not Equal (Enum)"!(A:EFortActivityViewModelDownloadState Enum=@n3, B:EFortActivityViewModelDownloadState Enum=Unknown) -> (ReturnValue:bool)
     B=Get IsContentGated() -> (IsContentGated:bool)
     C=GetIsCountdownTimerVisible!(self, TimerState:ECountdownTimerState Enum=@n2) -> (IsTimerVisible:bool)
    ))
   n5 = GetIsLikeDislikeIslandOptionEnabled!(self:Fort Runtime Options Object Reference=GetRuntimeOptions()) -> (ReturnValue:bool)
   n6 = Get HasShowLikePercentFeatureTag() -> (HasShowLikePercentFeatureTag:bool)
   n7 = GetRuntimeOptions()
   n8 = OR Boolean!(
    A=NOT Boolean(A=AND Boolean(A=@n6, B=@n5))
    B=AND Boolean(
     A=Get ShouldShowCCUAndLikePercent() -> (ShouldShowCCUAndLikePercent:bool)
     B=GetShowCCUAndLikePercentTogetherOnTiles!(self:Fort Runtime Options Object Reference=@n7) -> (ReturnValue:bool)
    )
   ) -> (ReturnValue)
   n9 = "integer > integer"(A=Get CCUCount() -> (CCUCount:int=0))
   > Return Node!(
    execute
    Visibility:ESlateVisibility Enum=Select!(
     Option 0:ESlateVisibility Enum=Collapsed
     Option 1:ESlateVisibility Enum=SelfHitTestInvisible
     Index:bool=AND Boolean!(A=@n9, B=@n4, C:bool=@n1, D:bool=@n8) -> (ReturnValue)
    ) -> (ReturnValue:ESlateVisibility Enum)
   )
  ]}
 GetHighlightText {nodes [
   n1 = GetHighlightText!() -> (then, LabelText:Text, IsActivityLocked:bool=~, DoesActivityRequirePurchase:bool=~) {is=title K2Node_FunctionEntry}
   > Return Node!(
    execute
    Text:Text=SelectText(A="LOCTEXT(\"Requires Purchase\")", B=SelectText(B=@n1, bPickA=@n1.IsActivityLocked), bPickA=@n1.DoesActivityRequirePurchase)
   )
  ]}
 GetIsCountdownTimerVisible {nodes [
   n1 = GetIsCountdownTimerVisible!() -> (then, TimerState:ECountdownTimerState Enum=None) {is=title K2Node_FunctionEntry}
   > Return Node!(execute, IsTimerVisible:bool="Equal (Enum)"(A:ECountdownTimerState Enum=@n1, B:ECountdownTimerState Enum=Countdown))
  ]}
 GetLikePercentVisibility {nodes [
   GetLikePercentVisibility!() -> (
    then
    LikePercent:double=~
    DownloadState:EFortActivityViewModelDownloadState Enum=Unknown
    IsContentGated:bool=~
    TimerState:ECountdownTimerState Enum=None
    HasShowLikePercentFeatureTag:bool=~
    ShouldShowCCUAndLikePercent:bool=~
   ) {is=title K2Node_FunctionEntry}
   n1 = Get TimerState() -> (TimerState=None)
   n2 = Get DownloadState() -> (DownloadState=Unknown)
   n3 = NOT Boolean(A=OR Boolean!(
     A="Not Equal (Enum)"(A=@n2, B=Unknown)
     B=Get IsContentGated()
     C:bool=GetIsCountdownTimerVisible(TimerState=@n1) {is=title K2Node_CallFunction}
    ) -> (ReturnValue))
   n4 = GetIsLikeDislikeIslandOptionEnabled(self=GetRuntimeOptions())
   n5 = GetRuntimeOptions()
   n6 = OR Boolean!(
    A=Get HasShowLikePercentFeatureTag()
    B=AND Boolean!(A=Get ShouldShowCCUAndLikePercent(), B=GetShowCCUAndLikePercentTogetherOnTiles(self=@n5)) -> (ReturnValue)
   ) -> (ReturnValue)
   n7 = "float > float"(A=Get LikePercent() -> (LikePercent:double))
   > Return Node!(
    execute
    Visibility:ESlateVisibility Enum=Select(Option 0=Collapsed, Option 1=SelfHitTestInvisible, Index=AND Boolean!(A=@n7, B=@n3, C:bool=@n4, D:bool=@n6) -> (ReturnValue))
   )
  ]}
 GetNestedButtonGroupSelectionVM {nodes [
   GetNestedButtonGroupSelectionVM() {is=title K2Node_FunctionEntry}
   > Return Node!(
    execute
    ReturnValue:Fort Nested Button Group Selection VM Object Reference=Get FortNestedButtonGroupSelectionVM() -> (FortNestedButtonGroupSelectionVM:Fort Nested Button Group Selection VM Object Reference)
   )
  ]}
 GetPlaytimeBlocked {nodes [
   n1 = GetPlaytimeBlocked!() -> (then, PlaytimeState:EPlaytimeState Enum=Unblocked) {is=title K2Node_FunctionEntry}
   n2 = GetRuntimeOptions()
   > Return Node!(
    execute
    IsPlaytimeBlocked:bool=AND Boolean!(
     A="Equal (Enum)"(A:EPlaytimeState Enum=@n1, B:EPlaytimeState Enum=PlayTime_Blocked)
     B=Get bEnableThyme!(self:Fort Runtime Options Object Reference=@n2) -> (bEnableThyme:bool)
    ) -> (ReturnValue)
   )
  ]}
 GetPurchaseRequiredAndActivityLocked {nodes [
   n1 = GetPurchaseRequiredAndActivityLocked!() -> (then, PurchaseRequired:bool=~, ActivityLocked:bool=~) {is=title K2Node_FunctionEntry}
   > Return Node!(execute, IsPerchasedAndNotLocked:bool=AND Boolean(A=@n1, B=NOT Boolean(A=@n1.ActivityLocked)))
  ]}
 HandleTileSelectionActive {nodes [
   HandleTileSelectionActive!() -> (then, InActive:bool=~) {is=title K2Node_FunctionEntry}
   > n1 = Get!(execute:Exec) -> (then:Exec, else:Exec, FortNestedButtonGroupSelectionVM:Fort Nested Button Group Selection VM Object Reference)
   > SetIsSelected!(
    execute
    self:Fort Nested Button Group Selection VM Object Reference=@n1
    InButton:Fort Nested Button Group VMItem Interface=Self-Reference()
    bInSelected:bool=Get InActive() -> (InActive:bool)
   ) -> (then)
   n2 = Branch(execute=@n1.else, Condition=Get InActive())
   > OnTileSelectionActive()
   OnTileSelectionNotActive(execute=@n2.else)
  ]}
 InitializeClickedSound {nodes [
   InitializeClickedSound() {is=title K2Node_FunctionEntry}
   n1 = Get!() -> (PressedSlateSoundOverride_ResourceObject:Object Reference)
   > Is Valid(InputObject=@n1)
   > n2 = Cast To SoundBase!(execute, Object=@n1) -> (then, CastFailed, AsSound Base:Sound Base Object Reference) {is=title K2Node_DynamicCast}
   > Set Pressed Sound(Pressed Sound:Sound Base Object Reference=@n2) -> (Output_Get:Sound Base Object Reference)
   > SetPressedSoundOverride!(execute, self:Common Button Base Object Reference, Sound:Sound Base Object Reference) -> (then)
  ]}
 IsContentGated {nodes [
   n1 = IsContentGated!() -> (then, IsActivityLocked:bool=~, IsActivityRestricted:bool=~) {is=title K2Node_FunctionEntry}
   > Return Node!(execute, IsContentGated:bool=OR Boolean(A=@n1, B=@n1.IsActivityRestricted))
  ]}
 IsUsingFocusNavigationOnKBM {nodes [
   IsUsingFocusNavigationOnKBM() {is=title K2Node_FunctionEntry}
   n1 = GetCurrentInputType(self=Get CommonInputSubsystem())
   > Return Node!(
    execute
    IsUsingFocusNavigationForOutlineVisuals:bool=AND Boolean(
     A="Equal (Enum)"(A:ECommonInputType Enum=@n1, B:ECommonInputType Enum=MouseAndKeyboard)
     B=Get IsUsingFocusForOutlineVisuals() -> (IsUsingFocusForOutlineVisuals:bool)
    )
   )
  ]}
 IsValidActivity {nodes [
   Inputs!()
   Outputs!(IsValid:bool="Equal (Enum)"(
     A:EFortInvalidActivityReason Enum=Get InvalidActivityReason() -> (InvalidActivityReason:EFortInvalidActivityReason Enum=None)
     B:EFortInvalidActivityReason Enum=None
    ))
  ]}
 MarkTileAsSeen {nodes [
   MarkTileAsSeen() {is=title K2Node_FunctionEntry}
   > Branch(Condition=GetDiscoveryReduxExperimentValue!(FortPlayerController:Player Controller Object Reference=GetOwningPlayer()) -> (ReturnValue:bool))
   > MarkTileAsSeen(self:Fort Discover Link Entry Tile Item VM Object Reference=Get FortActivityViewModel()) {is=title K2Node_CallFunction}
  ]}
 OnByEpicUpdated {nodes [
   n1 = OnByEpicUpdated!() -> (then, IsByEpic:bool=~) {is=title K2Node_FunctionEntry}
   > SetVisibility(self=Get ByEpicLogo(), InVisibility=Select(Option 0=Collapsed, Option 1=SelfHitTestInvisible, Index=@n1))
   > SetDisabledThumbnailOverlay()
  ]}
 OnFocusGained {nodes [OnFocusGained!() -> (then, FocusedItem:WBP Island Tile Minimal Object Reference) {is=title K2Node_FunctionEntry}]}
 OnIsContentGatedUpdated {nodes [
   n1 = OnIsContentGatedUpdated!() -> (then, IsContentGated:bool=~) {is=title K2Node_FunctionEntry}
   > SetVisibility(self=Get SqueegeeSlot_Lock(), InVisibility=Select(Option 0=Collapsed, Option 1=SelfHitTestInvisible, Index=@n1))
   > SetupSqueegeeWidgets!(execute, self, ActivityVM:Discover Activity View Model Object Reference=Get FortActivityViewModel()) -> (then)
  ]}
 OnMouseButtonDown {nodes [
   n1 = OnMouseButtonDown!() -> (then, MyGeometry:Geometry Structure, MouseEvent:"Pointer Event Structure (by ref)") {is=title K2Node_FunctionEntry}
   > n2 = OnRightClickMouse!(execute, self:WBP Discover Tile Base Object Reference=Get WBP_Discover_Tile_Base(), MouseEvent:Pointer Event Structure=@n1.MouseEvent) -> (then, WasHandled:Event Reply Structure)
   > Return Node!(execute, ReturnValue:Event Reply Structure=@n2)
  ]}
 OnSelectedButtonChanged {nodes [
   n1 = OnSelectedButtonChanged!() -> (then, SelectedButton:Fort Nested Button Group VMItem Interface) {is=title K2Node_FunctionEntry}
   > n2 = Branch(Condition="Equal (Object)"(A:Fort Nested Button Group VMItem Interface=@n1, B=Self-Reference()))
   n3 = Get IsTileActive!() -> (IsTileActive)
   > n4 = Branch(Condition=@n3)
   OnTileSelectionActive(execute=@n4.else)
   Branch(execute=@n2.else, Condition=@n3)
   > OnTileSelectionNotActive()
   > Branch(Condition=AND Boolean(A=GetSelected!(self:Common Button Base Object Reference) -> (ReturnValue:bool), B=Get bSelectable()))
   > SetSelectedInternal!(execute, self:Common Button Base Object Reference, bInSelected:bool=false, bAllowSound:bool=false, bBroadcast:bool=true) -> (then)
  ]}
 OnSetInvalidActivityReason {nodes [
   OnSetInvalidActivityReason!() -> (then, InvalidActivityReason:EFortInvalidActivityReason Enum=None) {is=title K2Node_FunctionEntry}
   > SetDisabledThumbnailOverlay()
  ]}
 OnTileSelectionActive {nodes [
   OnTileSelectionActive() {is=title K2Node_FunctionEntry}
   > Set IsTileActive(IsTileActive:bool=true) -> (Output_Get:bool)
   > SetScrollingEnabled(self=Get Text_Highlight(self=Get Highlight_Label()), bInIsScrollingEnabled=true)
   > OnTileBaseSelectionActive(self:WBP Discover Tile Base Object Reference=Get WBP_Discover_Tile_Base())
   > SetVisibility(self=Get SqueegeeSlot_Tag() -> (SqueegeeSlot_Tag:Squeegee Injection Slot Object Reference), InVisibility=SelfHitTestInvisible)
   > PlayAnimationForward(InAnimation=Get Anim_OnHoverUnhoverRehover() -> (Anim_OnHoverUnhoverRehover:Widget Animation Object Reference))
   > MarkTileAsSeen(self:Self Object Reference)
  ]}
 OnTileSelectionNotActive {nodes [
   OnTileSelectionNotActive() {is=title K2Node_FunctionEntry}
   > Set IsTileActive(IsTileActive=false)
   > SetScrollingEnabled(self=Get Text_Highlight(self=Get Highlight_Label()), bInIsScrollingEnabled=false)
   > OnTileBaseSelectionNotActive(self:WBP Discover Tile Base Object Reference=Get WBP_Discover_Tile_Base())
   > SetVisibility(self=Get SqueegeeSlot_Tag(), InVisibility=Collapsed)
   > PlayAnimationReverse(InAnimation=Get Anim_OnHoverUnhoverRehover())
   > Branch(Condition=Get IsOldDiscoverCode())
   > ResetAnimationState(self:WBP Discover Tile Base Object Reference=Get WBP_Discover_Tile_Base())
  ]}
 PlayPressedSound {nodes [
   PlayPressedSound() {is=title K2Node_FunctionEntry}
   > n1 = Get!(execute:Exec) -> (then:Exec, else:Exec, Pressed Sound:Sound Base Object Reference)
   > SpawnSound2D!(
    execute
    Sound:Sound Base Object Reference=@n1
    VolumeMultiplier:float=1
    PitchMultiplier:float=1
    StartTime:float=0
    ConcurrencySettings:Sound Concurrency Object Reference
    bPersistAcrossLevelTransition:bool=false
    bAutoDestroy:bool=true
   ) -> (then, ReturnValue:Audio Component Object Reference)
  ]}
 Set Icon Mobile Multiplier {nodes [
   n1 = Set Icon Mobile Multiplier!() -> (then, ImageObject:Image Object Reference) {is=title K2Node_FunctionEntry}
   > n2 = GetScalarParameterValue!(
    execute
    Collection:Material Parameter Collection Object Reference=/DiscoveryBrowser/DataAssets/MobileSharedValues
    ParameterName:Name=MobileIconSizeMultiplier
   ) -> (then, ReturnValue:float) {is=title K2Node_CallMaterialParameterCollectionFunction}
   > Set L_MobileSizeIconMultiplier(L_MobileSizeIconMultiplier:float=@n2) -> (Output_Get:float)
   n3 = IsMobileGame()
   n4 = Break Slate Brush!(SlateBrush:"Slate Brush Structure (by ref)"=Get Brush!(self:Image Object Reference=@n1) -> (Brush:Slate Brush Structure)) -> (ImageSize_X:double, ImageSize_Y:double) {is=title K2Node_BreakStruct}
   n5 = Get L_MobileSizeIconMultiplier() -> (L_MobileSizeIconMultiplier:float)
   > n6 = Set members in Slate Brush!(
    execute:Exec
    StructRef:"Slate Brush Structure (by ref)"=Get Brush(self=@n1)
    ImageSize_X:double=Select(Option 0:double=@n4, Option 1:double=float * float(A=@n4, B=@n5) {is=title K2Node_PromotableOperator}, Index=@n3) -> (ReturnValue:double)
    ImageSize_Y:double=Select(Option 0=@n4.ImageSize_Y, Option 1=float * float(A=@n4.ImageSize_Y, B=@n5), Index=@n3)
   ) -> (then:Exec, StructOut:"Slate Brush Structure (by ref)") {is=title K2Node_SetFieldsInStruct}
   > Set Brush!(execute, Brush:Slate Brush Structure=@n6, self:Image Object Reference=@n1) -> (then, Output_Get:Slate Brush Structure)
  ]}
 Set Image Spinner State {nodes [
   n1 = Set Image Spinner State!() -> (then, InImageState:EActivityImageLoadingState Enum=Unset) {is=title K2Node_FunctionEntry}
   > PlayAnimationReverse(InAnimation=Get Anim_OnLoadedImage() -> (Anim_OnLoadedImage:Widget Animation Object Reference))
   "Equal (Enum)"(A:EActivityImageLoadingState Enum=@n1, B:EActivityImageLoadingState Enum=Loading)
  ]}
 Set Image State {nodes [
   n1 = Set Image State!() -> (then, LoadingEnum:EActivityImageLoadingState Enum=Unset) {is=title K2Node_FunctionEntry}
   > SetImageErrorState!(execute, self, ImageLoadState:EActivityImageLoadingState Enum=@n1) -> (then)
   > Set Image Spinner State!(execute, self, InImageState:EActivityImageLoadingState Enum=@n1) -> (then) {is=title K2Node_CallFunction}
   > n2 = Switch on EActivityImageLoadingState!(execute, Selection:EActivityImageLoadingState Enum=@n1) -> (Unset:Exec, Loading:Exec, Loaded:Exec, Error:Exec) {is=title K2Node_SwitchEnum}
   Set Image Success State(execute=@n2.Loaded) {is=title K2Node_CallFunction}
  ]}
 Set Image Success State {nodes [
   Set Image Success State() {is=title K2Node_FunctionEntry}
   > n1 = GetDynamicMaterial(self=Get TileThumbnail() -> (TileThumbnail:Image Object Reference))
   > SetTextureParameterValue(
    self=@n1
    ParameterName=Get TextureParam() -> (TextureParam:Name=None)
    Value=GetImageTexture!(self:Fort Discover Link Entry Tile Item VM Object Reference=Get FortActivityViewModel()) -> (ReturnValue:Texture Object Reference)
   )
   > PlayAnimationForward(InAnimation=Get Anim_OnLoadedImage())
   > Update Texture Margin To Resolution!(execute, self, MaterialInstance:Material Instance Dynamic Object Reference=@n1) -> (then)
  ]}
 SetCCUCount {nodes [
   n1 = SetCCUCount!() -> (then, CCUCount:int) {is=title K2Node_FunctionEntry}
   n2 = Get WBP_Discover_Tile_Base()
   > n3 = Branch(Condition=Get IsTileNameInside!(self:WBP Discover Tile Base Object Reference=@n2) -> (IsTileNameInside:bool))
   > SetBadgeText!(
    execute
    self:WBP UIKit Badge Base Object Reference=Get Badge_CCUCount()
    Text:Text=SelectText(
     A=GetShorthandQuantity!(Quantity:int=@n1, MinimumFractionalDigits:int=0, MaximumFractionalDigits:int=1, bShouldRoundDown:bool=false) -> (ReturnValue:Text)
     B="To Text (Integer)"(Value=@n1)
     bPickA="integer > integer"(A=@n1, B=999)
    )
   ) -> (then)
   Update Secondary Text as CCU!(execute=@n3.else, self:WBP Discover Tile Base Object Reference=@n2, CCUCount:int=@n1) -> (then)
  ]}
 SetContextMenuFlowRiderTelemetry {nodes [
   n1 = SetContextMenuFlowRiderTelemetry!() -> (then, LinkID:Online Link Id Structure) {is=title K2Node_FunctionEntry}
   > n2 = LinkIdToString(LinkString=@n1)
   > SetParentIDForContextMenuTelemetry!(
    execute
    self:WBP UIKit Context Menu Anchor Button Base Object Reference=Get WBP_UIKit_ContextMenuAnchorButton(self=Get WBP_Discover_Tile_Base())
    ParentID:String=Append(A=Island ID:, B=@n2)
    IsParentIdEGPI:bool=false
   ) -> (then)
  ]}
 SetCornerRadius {nodes [
   n1 = SetCornerRadius!() -> (then, CornerRadius:Linear Color Structure=(R=20,G=20,B=20,A=20)) {is=title K2Node_FunctionEntry}
   > n2 = GetDynamicMaterial(self=Get TileThumbnail())
   > SetVectorParameterValue(self=@n2, ParameterName=Corner Radius, Value=@n1)
  ]}
 SetCountdownTimer {nodes [
   n1 = SetCountdownTimer!() -> (then, IsTimerVisible:bool=~) {is=title K2Node_FunctionEntry}
   > SetVisibility(
    self=Get Badge_Countdown()
    InVisibility=Select(Option 0:ESlateVisibility Enum=Collapsed, Option 1:ESlateVisibility Enum=HitTestInvisible, Index=@n1) -> (ReturnValue:ESlateVisibility Enum)
   )
  ]}
 SetCreatorId {nodes [
   n1 = SetCreatorId!() -> (then, InCreatorId:String) {is=title K2Node_FunctionEntry}
   n2 = GetActivitySelector()
   n3 = GetPlayerPageCreatorAccountId!(self:Fort Activity Selector Object Reference=@n2) -> (ReturnValue:String)
   > Branch(Condition=AND Boolean(A="Equal Exactly (String)"(A=@n1, B=@n3), B=IsShowingPlayerPage!(self:Fort Activity Selector Object Reference=@n2) -> (ReturnValue:bool)))
   > RemoveInteractionByName(self=Get WBP_Discover_Tile_Base(), ActionToRemove=Get ContextMenuViewCreatorName() -> (ContextMenuViewCreatorName=None))
  ]}
 SetCreatorLogoOverrideVisible {nodes [
   n1 = SetCreatorLogoOverrideVisible!() -> (then, CreatorLogoLoadingState:EActivityImageLoadingState Enum=Unset, CreatorLogoTexture:Texture Object Reference) {is=title K2Node_FunctionEntry}
   n2 = AND Boolean(A="Equal (Enum)"(A=@n1, B=Loaded), B=IsValid(Object=@n1.CreatorLogoTexture))
   > n3 = Branch(Condition=@n2)
   > n4 = Cast To Texture2DDynamic!(execute, Object=@n1.CreatorLogoTexture) -> (then, CastFailed, AsTexture 2DDynamic:Texture 2DDynamic Object Reference) {is=title K2Node_DynamicCast}
   > n5 = SetBrushFromTextureDynamic!(execute, self:Image Object Reference=Get CMS_CreatorLogo(), Texture:Texture 2DDynamic Object Reference=@n4, bMatchSize:bool=false) -> (then)
   Return Node!(
    execute=[@n5.then, @n3.else]
    VisibleWidget:Widget Object Reference=Select(Option 0:Widget Object Reference=Get ByEpicLogo(), Option 1:Widget Object Reference=Get CMS_CreatorLogo(), Index=@n2) -> (ReturnValue:Widget Object Reference)
   )
  ]}
 SetDisabledThumbnailOverlay {nodes [
   SetDisabledThumbnailOverlay() {is=title K2Node_FunctionEntry}
   > n1 = GetDynamicMaterial(self=Get TileThumbnail())
   n2 = Get FortActivityViewModel()
   n3 = DoesActivityRequirePurchase!(self:Discover Activity View Model Object Reference=@n2) -> (ReturnValue:bool)
   n4 = IsContentGated!(self:Fort Game Activity Object Reference=GetGameActivity(self=@n2)) -> (ReturnValue:bool) {is=title K2Node_CallFunction}
   n5 = AND Boolean!(
    A="Not Equal (Enum)"(A:EFortInvalidActivityReason Enum=Get InvalidActivityReason() -> (InvalidActivityReason=None), B:EFortInvalidActivityReason Enum=None)
    B="Equal (Boolean)"(A:bool=Get IsFullTile(), B:bool) -> (ReturnValue:bool=~) {is=title K2Node_PromotableOperator}
    C:bool="Equal (Boolean)"(A=Get bIsCurrentActivity!(self:Discover Activity View Model Object Reference=@n2) -> (bIsCurrentActivity:bool)) -> (ReturnValue=~)
   ) -> (ReturnValue)
   > SetScalarParameterValue(
    self=@n1
    ParameterName=Get ThumbnailDisabledOverlay() -> (ThumbnailDisabledOverlay:Name=None)
    Value="To Float (Boolean)"(InBool=OR Boolean!(A=@n4, B=@n3, C:bool=@n5) -> (ReturnValue))
   )
   > Set AllowTileClicks(AllowTileClicks:bool=NOT Boolean(A=@n5)) -> (Output_Get:bool)
  ]}
 SetDownloadProgressVisibility {nodes [
   n1 = SetDownloadProgressVisibility!() -> (then, IsVisible:bool=~) {is=title K2Node_FunctionEntry}
   > SetVisibility(
    self=Get StatusIndicator_DownloadProgress() -> (StatusIndicator_DownloadProgress:WBP Status Indicator Download Progress Object Reference)
    InVisibility=Select(Option 0:ESlateVisibility Enum=Collapsed, Option 1:ESlateVisibility Enum=HitTestInvisible, Index=@n1) -> (ReturnValue:ESlateVisibility Enum)
   )
  ]}
 SetFriendsPlayingCount {nodes [
   n1 = SetFriendsPlayingCount!() -> (then, FriendsPlayingCount:int) {is=title K2Node_FunctionEntry}
   n2 = "integer <= integer"!(A:int=@n1, B:int=0) -> (ReturnValue:bool) {is=title K2Node_CallFunction}
   n3 = Get DisableCornerInfo()
   > SetVisibility(
    self=Get FriendsPlaying() -> (FriendsPlaying:WBP UIKit Badge Object Reference)
    InVisibility=Select(Option 0=HitTestInvisible, Option 1=Collapsed, Index=OR Boolean!(A=@n2, B=@n3) -> (ReturnValue))
   )
   > SetBadgeText(
    self=Get FriendsPlaying()
    Text=SelectText(
     A=GetShorthandQuantity(Quantity=@n1, MinimumFractionalDigits=0, MaximumFractionalDigits=1, bShouldRoundDown=false)
     B="To Text (Integer)"(Value=@n1)
     bPickA="integer > integer"(A=@n1, B=999)
    )
   )
  ]}
 SetHighlightText {nodes [
   n1 = SetHighlightText!() -> (then, Text:Text) {is=title K2Node_FunctionEntry}
   > n2 = Set CurrentLabelText(CurrentLabelText:Text=@n1) -> (Output_Get:Text)
   > SetText!(execute, self:WBP UIKit Highlight Base Object Reference=Get Highlight_Label(), Text:Text=@n2) -> (then)
   > SetDisabledThumbnailOverlay() {is=title K2Node_CallFunction}
  ]}
 SetHighlightType {nodes [
   n1 = SetHighlightType!() -> (then, Type:EFortDiscoverLabelStyle Enum=Utility) {is=title K2Node_FunctionEntry}
   > Set CurrentLabelType(CurrentLabelType:EFortDiscoverLabelStyle Enum=@n1) -> (Output_Get:EFortDiscoverLabelStyle Enum=Utility)
   > SetType!(
    execute
    self:WBP UIKit Highlight Base Object Reference=Get Highlight_Label()
    Type:E_UI_HighlightType Enum=Byte to Enum E_UI_HighlightType!(Byte:Byte=MakeLiteralByte!(Value:Byte=@n1) -> (ReturnValue:Byte=0)) -> (ReturnValue:E_UI_HighlightType Enum) {is=title K2Node_CastByteToEnum}
   ) -> (then)
  ]}
 SetImageErrorState {nodes [
   n1 = SetImageErrorState!() -> (then, ImageLoadState:EActivityImageLoadingState Enum=Unset) {is=title K2Node_FunctionEntry}
   n2 = "Equal (Enum)"(A=@n1, B=Error)
   > SetVisibility(self=Get ErrorIcon() -> (ErrorIcon:Image Object Reference), InVisibility=Select(Option 0=Collapsed, Option 1=HitTestInvisible, Index=@n2))
   > n3 = GetDynamicMaterial(self=Get TileThumbnail())
   > SetScalarParameterValue(self=@n3, ParameterName=Override Loading Visiblity, Value=Select(Option 0:float=0, Option 1:float=1, Index=@n2) -> (ReturnValue:float))
  ]}
 SetInteractionAvailability {nodes [
   n1 = SetInteractionAvailability!() -> (then, InteractionID:Name, Availability:bool=~) {is=title K2Node_FunctionEntry}
   > SetInteractionAvailability!(
    execute
    self:UIKit Context Menu Action Context Object Reference=Get ContextMenuActionContext(self=Get WBP_Discover_Tile_Base())
    InteractionNameID:Name=@n1
    bInAvailability:bool=@n1.Availability
   ) -> (then) {is=title K2Node_CallFunction}
  ]}
 SetIsCurrentActivity {nodes [
   SetIsCurrentActivity!() -> (then, IsCurrentActivity:bool=~) {is=title K2Node_FunctionEntry}
   > SetDisabledThumbnailOverlay()
   > n1 = GetDynamicMaterial(self=Get TileThumbnail())
   > SetScalarParameterValue(self=@n1, ParameterName=Is Selected, Value="To Float (Boolean)"(InBool=Get IsCurrentActivity() -> (IsCurrentActivity:bool)))
  ]}
 SetIsFavorite {nodes [
   n1 = SetIsFavorite!() -> (then, IsFavorited:bool=~) {is=title K2Node_FunctionEntry}
   > SetVisibility(
    self=Get StatusIndicator_Favorite() -> (StatusIndicator_Favorite:WBP UIKit Status Indicator Object Reference)
    InVisibility=Select(Option 0:ESlateVisibility Enum=Collapsed, Option 1:ESlateVisibility Enum=HitTestInvisible, Index=@n1) -> (ReturnValue:ESlateVisibility Enum)
   )
   > PlayAnimation(InAnimation=Get Anim_OnFavorite() -> (Anim_OnFavorite:Widget Animation Object Reference))
   > n2 = Branch(Condition=IsPreviewTile!(self:Discover Activity View Model Object Reference=Get FortActivityViewModel()) -> (ReturnValue:bool))
   > RemoveInteractionByName(self=Get WBP_Discover_Tile_Base(), ActionToRemove=Get ContextMenuFavoriteName() -> (ContextMenuFavoriteName=None))
   Set Up Favorite Interaction!(execute=@n2.else, self, IsFavorited:bool=@n1) -> (then) {is=title K2Node_CallFunction}
  ]}
 SetIsPreviewTile {nodes [
   n1 = SetIsPreviewTile!() -> (then, IsPreviewTile:bool=~) {is=title K2Node_FunctionEntry}
   > n2 = Branch(Condition=@n1)
   > RemoveAllInteractions(self:WBP Discover Tile Base Object Reference=Get WBP_Discover_Tile_Base())
   n3 = Get WBP_Discover_Tile_Base()
   n4 = Branch(
    execute=@n2.else
    Condition=HasInteractionItems!(self:UIKit Context Menu Action Context Object Reference=Get ContextMenuActionContext(self=@n3)) -> (ReturnValue:bool)
   )
   AddContextMenuItems(execute=@n4.else)
  ]}
 SetIsTileEnabled {nodes [
   SetIsTileEnabled() {is=title K2Node_FunctionEntry}
   > Set bIsEnabled(bIsEnabled:bool=false) -> (Output_Get:bool)
   n1 = Get FortActivityViewModel()
   n2 = Get Invalid Activity Reason!(self:Fort Activity Tile Details Display Object Reference) -> (ReturnValue:EFortInvalidActivityReason Enum=None) {is=title K2Node_CallFunction}
   n3 = Get IsFullTile()
   n4 = Get bIsCurrentActivity(self=@n1)
   OR Boolean!(
    A=IsActivityLocked!(self:Fort Discover Link Entry Tile Item VM Object Reference=@n1) -> (ReturnValue:bool)
    B=DoesActivityRequirePurchase(self=@n1)
    C:bool=AND Boolean(A="Not Equal (Enum)"(A=@n2, B=None), B="Equal (Boolean)"(A=@n3) -> (ReturnValue=~), C="Equal (Boolean)"(A=@n4) -> (ReturnValue=~))
   ) -> (ReturnValue)
  ]}
 SetIslandDescription {nodes [SetIslandDescription!() -> (then, IslandDescription:Text) {is=title K2Node_FunctionEntry}]}
 SetIslandName {nodes [SetIslandName!() -> (then, IslandName:Text) {is=title K2Node_FunctionEntry}]}
 SetLabelVisiblity {nodes [SetLabelVisiblity!() -> (then, HasUnseenNotification:bool=~) {is=title K2Node_FunctionEntry}]}
 SetLikePercent {nodes [
   n1 = SetLikePercent!() -> (then, LikePercent:double=~) {is=title K2Node_FunctionEntry}
   n2 = Get WBP_Discover_Tile_Base()
   > n3 = Branch(Condition=Get IsTileNameInside(self=@n2))
   n4 = Round(A=@n1) {is=title K2Node_CallFunction}
   > SetBadgeText(self=Get Badge_LikePercent(), Text=Format Text!(Format="LOCTEXT(\"{p}%\")", p:Text="To Text (Integer)"(Value=@n4)) -> (Result))
   Update Secondary Text as Like Percent!(execute=@n3.else, self:WBP Discover Tile Base Object Reference=@n2, LikePercent:int=@n4) -> (then)
  ]}
 SetLiveLabelTextAndType {nodes [SetLiveLabelTextAndType!() -> (then, TimerState:ECountdownTimerState Enum=None) {is=title K2Node_FunctionEntry}]}
 SetLogoMuted {nodes [
   n1 = SetLogoMuted!() -> (then, IsMuted:bool=~) {is=title K2Node_FunctionEntry}
   > SetRenderOpacity(
    self=Get Switcher_Logo() -> (Switcher_Logo:Common Visibility Switcher Object Reference)
    InOpacity=SelectFloat(A=Get Opacity_Locked() -> (Opacity_Locked:double), B=1, bPickA=@n1)
   )
  ]}
 SetMobileLayout {nodes [
   SetMobileLayout() {is=title K2Node_FunctionEntry}
   > Branch(Condition=IsMobileGame())
   > SetBadgeSize!(
    execute
    self:WBP UIKit Badge Base Object Reference=[Get Badge_LikePercent(), Get Badge_CCUCount()]
    BadgeSize:E_UI_BadgeSize Enum=NewEnumerator1
   ) -> (then)
   > UpdateSize!(
    execute
    self:WBP UIKit Status Indicator Base Object Reference=Get StatusIndicator_Favorite()
    IconSize:E_UI_StatusIndicator_Size Enum=NewEnumerator3
   ) -> (then)
  ]}
 SetPlaytimeBlocked {nodes [
   n1 = SetPlaytimeBlocked!() -> (then, IsPlaytimeOut:bool=~) {is=title K2Node_FunctionEntry}
   > Set IsPlayTimeOut(IsPlayTimeOut:bool=@n1) -> (Output_Get:bool)
   > AddContextMenuItems()
  ]}
 SetSlotOutsideAlignment {nodes [
   n1 = SetSlotOutsideAlignment!() -> (then, SlotOutsideAlignment:E_UI_Alignment Enum=NewEnumerator0) {is=title K2Node_FunctionEntry}
   > Set SlotOutsideAlignment(SlotOutsideAlignment:E_UI_Alignment Enum=@n1) -> (Output_Get:E_UI_Alignment Enum=NewEnumerator0)
   > SetGridSlotColumnAndRow!(
    execute
    Widget:Widget Object Reference=Get Slot_OutsideContent()
    Row:int=Select!(NewEnumerator0:int=0, NewEnumerator1:int=2, NewEnumerator2:int=1, NewEnumerator3:int=1, Index:E_UI_Alignment Enum=@n1) -> (ReturnValue:int)
    Row Span:int=1
    Column:int=Select(NewEnumerator0=1, NewEnumerator1=1, NewEnumerator2=0, NewEnumerator3=2, Index=@n1)
    Column Span:int=1
   ) -> (then)
  ]}
 SetThumbnailRestricted {nodes [
   n1 = SetThumbnailRestricted!() -> (then, IsRestricted:bool=~) {is=title K2Node_FunctionEntry}
   > n2 = GetDynamicMaterial(self=Get TileThumbnail())
   > SetScalarParameterValue(
    self=@n2
    ParameterName=Saturation
    Value=SelectFloat(A=Get ThumbnailSaturation_Restricted() -> (ThumbnailSaturation_Restricted:double), B=1, bPickA=@n1)
   )
  ]}
 SetupCountdownTimerViewModel {nodes [
   n1 = SetupCountdownTimerViewModel!() -> (then, DiscoverActivityViewModel:Discover Activity View Model Object Reference) {is=title K2Node_FunctionEntry}
   n2 = Get FortCountdownTimerVM() -> (FortCountdownTimerVM:Countdown Timer View Model Object Reference)
   > SetupViewModel!(
    execute
    self:Countdown Timer View Model Object Reference=@n2
    InFutureTime:"Date Time Structure (by ref)"=Select!(
     Option 0:Date Time Structure=GetNextStartTime!(self:Fort Discover Link Entry Tile Item VM Object Reference=@n1) -> (ReturnValue:Date Time Structure)
     Option 1:Date Time Structure=GetNextDiscoverTileCountdownTimerEnd!(self:Fort Discover Link Entry Tile Item VM Object Reference=@n1) -> (ReturnValue:Date Time Structure)
     Index:bool=HasActiveDiscoverTileCountdownTimer!(self:Fort Discover Link Entry Tile Item VM Object Reference=@n1) -> (ReturnValue:bool)
    ) -> (ReturnValue:Date Time Structure)
    InLiveText:"Text (by ref)"=Format Text!(Format="LOCTEXT(\"Live\")") -> (Result)
   ) -> (then)
   > SetTimerUnitDisplayType!(execute, self:Countdown Timer View Model Object Reference=@n2, InTimerUnitDisplayType:ETimespanUnitDisplayFormat Enum=Abbreviated) -> (then)
   > SetApproximateLongTimers!(execute, self:Countdown Timer View Model Object Reference=@n2, bInApproximateLongTimers:bool=false) -> (then)
   > StartTimer(self:Countdown Timer View Model Object Reference=@n2)
  ]}
 SetupSqueegeeWidgets {nodes [
   n1 = SetupSqueegeeWidgets!() -> (then, ActivityVM:Discover Activity View Model Object Reference) {is=title K2Node_FunctionEntry}
   > ShowActivityDataFromViewModel!(
    execute
    self:Fort Squeegee Widget Injection Slot Object Reference=[Get SqueegeeSlot_Tag(), Get SqueegeeSlot_Lock()]
    GameActivityViewModel:Fort Discover Link Entry Tile Item VM Object Reference=@n1
   ) -> (then)
  ]}
 ToggleSelect {nodes [ToggleSelect!() -> (then, IsCurrentActivity:bool=~) {is=title K2Node_FunctionEntry}]}
 ToggleSpacer {nodes [ToggleSpacer!() -> (then, ShouldShowSpacer:bool=~) {is=title K2Node_FunctionEntry}]}
 Update Context Menu Why You Are Seeing This Modal {nodes [
   n1 = Update Context Menu Why You Are Seeing This Modal!() -> (then, ShouldAddInfoModal:bool=~) {is=title K2Node_FunctionEntry}
   > RemoveInteractionByName(self=Get WBP_Discover_Tile_Base(), ActionToRemove=Get RecommendModalName() -> (RecommendModalName=None))
   > Branch(Condition=@n1)
   > AddInfoModalToContextMenu() {is=title K2Node_CallFunction}
  ]}
 Update Texture Margin To Resolution {nodes [
   n1 = Update Texture Margin To Resolution!() -> (then, MaterialInstance:Material Instance Dynamic Object Reference) {is=title K2Node_FunctionEntry}
   > n2 = Get Scalar Parameter Value!(execute, self:Material Instance Dynamic Object Reference=@n1, ParameterName:Name=Media Width) -> (then, ReturnValue:float) {is=title K2Node_CallFunction}
   > n3 = Get Scalar Parameter Value(self=@n1, ParameterName=Media Height)
   > Branch(Condition=AND Boolean!(A="float > float"(A=@n3), B="float > float"(A=@n2)) -> (ReturnValue))
   n4 = Get FortActivityViewModel()
   n5 = NOT Boolean(A=AND Boolean(A=IsEpicCreated!(self:Fort Discover Link Entry Tile Item VM Object Reference=@n4) -> (ReturnValue:bool), B=Get IsFullTile()))
   n6 = BreakVector2D(InVec=GetTextureSize!(self:Fort Discover Link Entry Tile Item VM Object Reference=Get FortActivityViewModel()) -> (ReturnValue:Vector 2D Structure))
   n7 = "Division (Whole and Remainder)"!(Dividend:double=@n6, Divisor:double=@n2) -> (Remainder:double, ReturnValue:int=0)
   n8 = "Division (Whole and Remainder)"(Dividend=@n6.Y, Divisor=@n3) -> (ReturnValue=0)
   n9 = NOT Boolean(A=AND Boolean(A="Equal (Float)"(A=@n7), B="Equal (Float)"(A=@n8)))
   > SetPadding!(
    execute
    self:Overlay Slot Object Reference=SlotAsOverlaySlot(Widget=Get TileThumbnail())
    InPadding:Margin Structure=Select(
     Option 0:Margin Structure=Make Margin(Left=0, Top=0, Right=0, Bottom=0)
     Option 1:Margin Structure=Make Margin(Left=1.5, Top=1.5, Right=1.5, Bottom=1.5)
     Index=AND Boolean(A=@n9, B=@n5)
    ) -> (ReturnValue:Margin Structure)
   ) -> (then)
  ]}
 UpdateContextMenuNotInterested {nodes [
   n1 = UpdateContextMenuNotInterested!() -> (then, ShouldAddNotInterested:bool=~) {is=title K2Node_FunctionEntry}
   > RemoveInteractionByName(self=Get WBP_Discover_Tile_Base(), ActionToRemove=Get ContextMenuNotInterestedName() -> (ContextMenuNotInterestedName=None))
   > Branch(Condition=@n1)
   > SetUpNotInterestedInteraction() {is=title K2Node_CallFunction}
  ]}
 ViewAllTileClicked {nodes [ViewAllTileClicked() {is=title K2Node_FunctionEntry}]}
]
variables [
 AllowTileClicks {type=boolean, default=true}
 ContextMenuCancelDownloadName {type=string, default=CancelDownload}
 ContextMenuDownloadFirstName {type=string, default=DownloadFirst}
 ContextMenuFavoriteName {type=string, default=ToggleFavorite}
 ContextMenuGoToLobbyName {type=string, default=GoToLobby}
 ContextMenuNotInterestedName {type=string, default=NotInterested}
 ContextMenuViewCreatorName {type=string, default=MoreFromCreator}
 ContextMenuViewDetails {type=string, default=ViewDetails}
 CurrentLabelText {type=string, default=""}
 CurrentLabelType {type=EFortDiscoverLabelStyle, default=Utility}
 DisableCornerInfo {type=boolean}
 Favorited Sound {type=SoundBase, default=/Game/Sounds/Fort_UI_Sounds/FrontEnd/MultiProduct/Common/Buttons/MSS_UI_MP_Button_Click_Favorite}
 HandlesOwnStartVisibilityImpression {type=boolean, default=true}
 IsFullTile {type=boolean}
 IsOldDiscoverCode {type=boolean}
 IsPlayTimeOut {type=boolean}
 IsRecommendedModal {type=boolean}
 IsTileActive {type=boolean}
 IsTileInCreatorPage {type=boolean}
 IsUsingFocusForOutlineVisuals {type=boolean}
 OnFocusGained {}
 Opacity_Locked {type=number, default=0.5}
 Pressed Sound {type=SoundBase, default=null}
 RecommendModalName {type=string, default=RecommendModal}
 Right Clicked Sound Override {type=SoundBase, default=null}
 SlotOutsideAlignment {type=E_UI_Alignment, default=Right}
 TextureParam {type=string, default=Thumbnail}
 ThumbnailDisabledOverlay {type=string, default=Overlay Opacity}
 ThumbnailSaturation_Restricted {type=number, default=0.25}
 Unfavorited Sound {type=SoundBase, default=/Game/Sounds/Fort_UI_Sounds/FrontEnd/MultiProduct/Common/Buttons/MSS_UI_MP_Button_Click_Unfavorite}
 ViewAllTileClicked {}
]
widgets {
 info {
  inheritedWidgetCount=8
  namedSlotCount=1
  parentClass=/Script/DiscoveryBrowserUI.FortActivityTileDetailsDisplay
  rootWidgetClass=&Overlay
  widgetCount=36
 }
 widgets [
  Overlay {widgetClassPath=&Overlay, bIsVariable=true, widget=Overlay}
  GridRoot_9Grid {
   widgetClassPath=&GridPanel
   bIsVariable=true
   parent=Overlay
   widget=GridRoot_9Grid
   properties {columnFill [0,1,0], rowFill [0,1,0]}
   slot {class=OverlaySlot}
  }
  WBP_Discover_Tile_Base {
   widgetClassPath=/DiscoveryBrowser/Tiles/WBP_Discover_Tile_Base
   bIsVariable=true
   parent=GridRoot_9Grid
   widget=WBP_Discover_Tile_Base
   properties {isTileNameInside=true, useIconBlock=true, useSecondaryTextBlock=true}
   slot {class=GridSlot, properties {column=1, columnSpan=1, layer=1, row=1, rowSpan=1}}
  }
  Grid_Main {
   widgetClassPath=&GridPanel
   namedSlotHost=WBP_Discover_Tile_Base
   widget=Grid_Main
   properties {clipping=ClipToBounds, columnFill [1], rowFill [1]}
  }
  ThumbnailOverlay {widgetClassPath=&Overlay, parent=Grid_Main, widget=ThumbnailOverlay, slot {class=GridSlot}}
  TileThumbnail {
   widgetClassPath=&Image
   bIsVariable=true
   parent=ThumbnailOverlay
   widget=TileThumbnail
   properties {
    visibility=SelfHitTestInvisible
    brush {
     bIsDynamicallyLoaded=false
     drawAs=Image
     imageType=NoImage
     mirroring=NoMirror
     resourceName=None
     resourceObject=/DiscoveryBrowser/Tiles/Islands/Materials/MI_UI_IslandTile
     tiling=NoTile
     imageSize=<32 32>
     margin=<ltrb 0 0 0 0>
     outlineSettings {
      bUseBrushTransparency=false
      roundingType=HalfHeightRadius
      width=0
      color {colorUseRule=UseColor_Specified, specifiedColor=<rgba 0 0 0 0>}
      cornerRadii=<0 0 0 0>
     }
     tintColor {colorUseRule=UseColor_Specified, specifiedColor=<rgba 1 1 1 1>}
     uVRegion {bIsValid=false, max=<0 0>, min=<0 0>}
    }
   }
   slot {class=OverlaySlot, properties {horizontalAlignment=HAlign_Fill, verticalAlignment=VAlign_Fill}}
  }
  ErrorIcon {
   widgetClassPath=&Image
   bIsVariable=true
   parent=ThumbnailOverlay
   widget=ErrorIcon
   properties {
    renderOpacity=0.2
    visibility=HitTestInvisible
    brush {
     bIsDynamicallyLoaded=false
     drawAs=Image
     imageType=NoImage
     mirroring=NoMirror
     resourceName=None
     resourceObject=/FortUILibrary/Textures/MaterialInstances/MI_UIKit_Icon_ImageError
     tiling=NoTile
     imageSize=<64 64>
     margin=<ltrb 0 0 0 0>
     outlineSettings {
      bUseBrushTransparency=false
      roundingType=HalfHeightRadius
      width=0
      color {colorUseRule=UseColor_Specified, specifiedColor=<rgba 0 0 0 0>}
      cornerRadii=<0 0 0 0>
     }
     tintColor {colorUseRule=UseColor_Specified, specifiedColor=<rgba 1 1 1 1>}
     uVRegion {bIsValid=false, max=<0 0>, min=<0 0>}
    }
   }
   slot {class=OverlaySlot, properties {horizontalAlignment=HAlign_Center, verticalAlignment=VAlign_Center}}
  }
  Text_DebugId {
   widgetClassPath=&TextBlock
   bInherited=true
   parent=Grid_Main
   widget=Text_DebugId
   properties {
    autoWrapText=true
    justification=Center
    text=EventId: epicgames_Some_Event_Id EventWindowId: Some_event_window_id EventWindow Playlist: Playlists_name_example
    visibility=Collapsed
    wrapTextAt=400
    wrappingPolicy=AllowPerCharacterWrapping
    font {
     bForceMonospaced=false
     bMaterialIsStencil=false
     fontMaterial=null
     fontObject=&Roboto
     letterSpacing=0
     monospacedWidth=1
     size=16
     skewAmount=0
     typefaceFontName=Bold
     outlineSettings {
      bApplyOutlineToDropShadows=false
      bMiteredCorners=false
      bSeparateFillAlpha=false
      outlineMaterial=null
      outlineSize=3
      outlineColor=<rgba 0 0 0 1>
     }
    }
   }
   slot {class=GridSlot, properties {layer=20}}
  }
  HorizontalBox_Bottom {
   widgetClassPath=&HorizontalBox
   parent=Grid_Main
   widget=HorizontalBox_Bottom
   slot {class=GridSlot, properties {layer=50, verticalAlignment=VAlign_Bottom, padding=<ltrb 8 0 8 8>}}
  }
  VerticalBox_TitleAndCCU {
   widgetClassPath=&VerticalBox
   bIsVariable=true
   parent=HorizontalBox_Bottom
   widget=VerticalBox_TitleAndCCU
   properties {clipping=ClipToBounds}
   slot {class=HorizontalBoxSlot, properties {verticalAlignment=VAlign_Bottom, size {sizeRule=Fill, value=1}}}
  }
  SizeBox_0 {
   widgetClassPath=&SizeBox
   parent=VerticalBox_TitleAndCCU
   widget=SizeBox_0
   properties {bOverride_MaxDesiredHeight=true, clipping=ClipToBounds, maxDesiredHeight=48}
   slot {class=VerticalBoxSlot, properties {verticalAlignment=VAlign_Top}}
  }
  NamedSlot_Title {widgetClassPath=&NamedSlot, bIsVariable=true, parent=SizeBox_0, widget=NamedSlot_Title, slot {class=SizeBoxSlot}}
  HorizontalBox_1 {
   widgetClassPath=&HorizontalBox
   parent=VerticalBox_TitleAndCCU
   widget=HorizontalBox_1
   slot {class=VerticalBoxSlot, properties {horizontalAlignment=HAlign_Left, verticalAlignment=VAlign_Bottom, padding=<ltrb 0 4 0 0>}}
  }
  Badge_Countdown {
   widgetClassPath=/FortUILibrary/UIKit/Badge/WBP_UIKit_Badge
   bIsVariable=true
   parent=HorizontalBox_1
   widget=Badge_Countdown
   properties {
    badgeText="{DD HH MM}"
    badgeType=Text
    indicator Type=Utility
    indicatorConfig=Icon
    indicatorIcon=/FortUILibrary/Textures/MaterialInstances/MI_UIKit_Icon_UpdatePending
   }
   slot {class=HorizontalBoxSlot, properties {horizontalAlignment=HAlign_Left, verticalAlignment=VAlign_Bottom, padding=<ltrb 0 0 8 0>}}
  }
  StatusIndicator_DownloadProgress {
   widgetClassPath=/FortUILibrary/UIKit/ProgressBar/WBP_StatusIndicator_DownloadProgress
   bIsVariable=true
   parent=HorizontalBox_1
   widget=StatusIndicator_DownloadProgress
   properties {visibility=Collapsed}
   slot {class=HorizontalBoxSlot, properties {horizontalAlignment=HAlign_Left, verticalAlignment=VAlign_Bottom, padding=<ltrb 0 -4 4 0>}}
  }
  Badge_CCUCount {
   widgetClassPath=&WBP_UIKit_Badge
   bIsVariable=true
   parent=HorizontalBox_1
   widget=Badge_CCUCount
   properties {badgeType=Text, indicatorConfig=None, visibility=Collapsed}
   slot {class=HorizontalBoxSlot, properties {horizontalAlignment=HAlign_Left, verticalAlignment=VAlign_Bottom, padding=<ltrb 0 0 4 0>}}
  }
  Badge_LikePercent {
   widgetClassPath=&WBP_UIKit_Badge
   bIsVariable=true
   parent=HorizontalBox_1
   widget=Badge_LikePercent
   properties {
    badgeType=Text
    indicator Type=Utility
    indicatorConfig=Icon
    indicatorIcon=/FortUILibrary/Textures/MaterialInstances/MI_UIKit_Icon_Like
    visibility=Collapsed
   }
   slot {class=HorizontalBoxSlot, properties {horizontalAlignment=HAlign_Left, verticalAlignment=VAlign_Bottom, padding=<ltrb 0 0 4 0>}}
  }
  StatusIndicator_Favorite {
   widgetClassPath=/FortUILibrary/UIKit/StatusIndicator/WBP_UIKit_StatusIndicator
   bIsVariable=true
   parent=HorizontalBox_1
   widget=StatusIndicator_Favorite
   properties {
    type=Utility
    icon=/DiscoveryBrowser/Tiles/Islands/Materials/MI_UIKit_Icon_Favorite_Discover
    iconSize=Compact
    overrideStatusIndicatorVisibility=true
   }
   slot {class=HorizontalBoxSlot, properties {horizontalAlignment=HAlign_Left, verticalAlignment=VAlign_Bottom}}
  }
  SqueegeeSlot_Lock {
   widgetClassPath=/Game/UI/Plugin/Squeegee/SqueegeeInjectionSlot
   bIsVariable=true
   parent=HorizontalBox_Bottom
   widget=SqueegeeSlot_Lock
   properties {pluginId=Squeegee, pluginWidgetId=Lock Icon}
   slot {class=HorizontalBoxSlot}
  }
  HorizontalBox_Top {
   widgetClassPath=&HorizontalBox
   parent=Grid_Main
   widget=HorizontalBox_Top
   slot {class=GridSlot, properties {layer=10, verticalAlignment=VAlign_Top, padding=<ltrb 8 8 8 0>}}
  }
  Switcher_Logo {
   widgetClassPath=&CommonVisibilitySwitcher
   bIsVariable=true
   parent=HorizontalBox_Top
   widget=Switcher_Logo
   slot {class=HorizontalBoxSlot, properties {horizontalAlignment=HAlign_Left, verticalAlignment=VAlign_Top}}
  }
  ByEpicLogo {
   widgetClassPath=&Image
   bIsVariable=true
   parent=Switcher_Logo
   widget=ByEpicLogo
   properties {
    visibility=SelfHitTestInvisible
    brush {
     bIsDynamicallyLoaded=false
     drawAs=Image
     imageType=NoImage
     mirroring=NoMirror
     resourceName=None
     resourceObject=/DiscoveryBrowser/Tiles/Islands/Materials/MI_UI_Island_Logo
     tiling=NoTile
     imageSize=<32 32>
     margin=<ltrb 0 0 0 0>
     outlineSettings {
      bUseBrushTransparency=false
      roundingType=HalfHeightRadius
      width=0
      color {colorUseRule=UseColor_Specified, specifiedColor=<rgba 0 0 0 0>}
      cornerRadii=<0 0 0 0>
     }
     tintColor {colorUseRule=UseColor_Specified, specifiedColor=<rgba 1 1 1 1>}
     uVRegion {bIsValid=false, max=<0 0>, min=<0 0>}
    }
   }
   slot {class=CommonVisibilitySwitcherSlot, properties {padding=<ltrb -3 0 4 0>}}
  }
  CMS_CreatorLogo {
   widgetClassPath=&Image
   bIsVariable=true
   parent=Switcher_Logo
   widget=CMS_CreatorLogo
   properties {
    visibility=SelfHitTestInvisible
    brush {
     bIsDynamicallyLoaded=false
     drawAs=Image
     imageType=NoImage
     mirroring=NoMirror
     resourceName=None
     resourceObject=null
     tiling=NoTile
     imageSize=<48 48>
     margin=<ltrb 0 0 0 0>
     outlineSettings {
      bUseBrushTransparency=false
      roundingType=HalfHeightRadius
      width=0
      color {colorUseRule=UseColor_Specified, specifiedColor=<rgba 0 0 0 0>}
      cornerRadii=<0 0 0 0>
     }
     tintColor {colorUseRule=UseColor_Specified, specifiedColor=<rgba 1 1 1 1>}
     uVRegion {bIsValid=false, max=<0 0>, min=<0 0>}
    }
   }
   slot {class=CommonVisibilitySwitcherSlot, properties {horizontalAlignment=HAlign_Left, verticalAlignment=VAlign_Top, padding=<ltrb 0 0 4 0>}}
  }
  HorizontalBox_0 {
   widgetClassPath=&HorizontalBox
   parent=HorizontalBox_Top
   widget=HorizontalBox_0
   slot {class=HorizontalBoxSlot, properties {horizontalAlignment=HAlign_Right, verticalAlignment=VAlign_Top, size {sizeRule=Fill, value=1}}}
  }
  FriendsPlaying {
   widgetClassPath=&WBP_UIKit_Badge
   bIsVariable=true
   parent=HorizontalBox_0
   widget=FriendsPlaying
   properties {badgeText="{##}", visibility=Collapsed}
   slot {class=HorizontalBoxSlot, properties {padding=<ltrb 0 0 4 0>}}
  }
  Highlight_Label {
   widgetClassPath=/FortUILibrary/UIKit/Highlight/WBP_UIKit_Highlight
   bIsVariable=true
   parent=HorizontalBox_0
   widget=Highlight_Label
   properties {highlightSize=Compact, highlightText=new, highlightType=Utility}
   slot {
    class=HorizontalBoxSlot
    properties {horizontalAlignment=HAlign_Right, verticalAlignment=VAlign_Top, padding=<ltrb 4 0 0 0>, size {sizeRule=Fill, value=1}}
   }
  }
  SqueegeeSlot_Tag {
   widgetClassPath=&SqueegeeInjectionSlot
   bIsVariable=true
   parent=HorizontalBox_0
   widget=SqueegeeSlot_Tag
   properties {
    pluginId=Squeegee
    pluginWidgetId=RatingTag
    visibility=Collapsed
    styleParams {bShowDetails=true, fontSize=13, paddingSize=0.5, iconSize=<24 24>, iconSkewSize=<5 26>}
   }
   slot {class=HorizontalBoxSlot, properties {horizontalAlignment=HAlign_Right, verticalAlignment=VAlign_Top, padding=<ltrb 4 0 0 0>}}
  }
  ThumbnailOverlay_Video {widgetClassPath=&Overlay, parent=Grid_Main, widget=ThumbnailOverlay_Video, slot {class=GridSlot}}
  ActivityVideoCycleWidget {
   widgetClassPath=&/WBP_IslandTile_VideoCycle
   bInherited=true
   bIsVariable=true
   parent=ThumbnailOverlay_Video
   widget=ActivityVideoCycleWidget
   properties {movieRestartDelay=0.1, movieStartDelay=0.1, visibility=Hidden}
   slot {class=OverlaySlot, properties {horizontalAlignment=HAlign_Fill, verticalAlignment=VAlign_Fill}}
  }
  Slot_OutsideContent {
   widgetClassPath=&NamedSlot
   bIsVariable=true
   parent=GridRoot_9Grid
   widget=Slot_OutsideContent
   slot {class=GridSlot, properties {column=2, columnSpan=1, layer=2, row=1, rowSpan=1}}
  }
  Text_ActivityName {widgetClassPath=&CommonTextBlock, bInherited=true, widget=null}
  Text_PlayerCount {widgetClassPath=&CommonTextBlock, bInherited=true, widget=null}
  Button_Favorite {widgetClassPath=&CommonButtonBase, bInherited=true, widget=null}
  Button_Details {widgetClassPath=&CommonButtonBase, bInherited=true, widget=null}
  ActivityBrowserTag_EpicOriginal {widgetClassPath=/Script/FortniteUI.FortActivityBrowserTag, bInherited=true, widget=null}
  InputActionWidget {widgetClassPath=&CommonActionWidget, bInherited=true, widget=null}
 ]
}
