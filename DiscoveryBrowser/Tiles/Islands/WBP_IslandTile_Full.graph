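# Full island tile for the Discovery browser, extending WBP_IslandTile_Minimal with a title text and a locked-status widget.
# Selection toggles the title, primary text marquee and focus, lock hover state and a thumbnail hover overlay material animation.
# Island name feeds both the base tile and title, whose auto-wrap is disabled on mobile.
# A "new" highlight label shows for unseen notifications when the Discovery Redux experiment is enabled.
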WidgetBlueprint /DiscoveryBrowser/Tiles/Islands/WBP_IslandTile_Full
animations [Anim_Hover {
  binding_count=1
  length_seconds=0.2
  object=&self:Anim_Hover
  sequence {
   bindings [TileThumbnail {
     id=#1
     possessed_class=Image
     tracks [{
       type=MovieSceneWidgetMaterialTrack
       sections [{type=MovieSceneParameterSection, channels [Hover Overlay Opacity {type=MovieSceneScriptingFloatChannel, keys [[0,0],[12,1]]}]}]
      }]
    }]
   display_rate {numerator=60}
  }
 }]
graphs [
 EventGraph {nodes [
   Event Construct()
   > Parent: Construct()
   > SetMobileStyle() {is=title K2Node_CallFunction}
   Event Destruct()
   > Parent: Destruct()
   n1 = Event PreConstruct()
   > Parent: PreConstruct(IsDesignTime=@n1.IsDesignTime)
   > UpdatePrimaryTextBlock!(
    execute
    self:WBP Discover Tile Base Object Reference=Get WBP_Discover_Tile_Base() -> (WBP_Discover_Tile_Base:WBP Discover Tile Base Object Reference)
    UsePrimaryText:bool=true
   ) -> (then)
  ]}
 OnTileSelectionActive {nodes [
   OnTileSelectionActive() {is=title K2Node_FunctionEntry}
   > Parent: OnTileSelectionActive() {is=title K2Node_CallParentFunction}
   n1 = Get WBP_Discover_Tile_Base()
   > UpdatePrimaryTextScrolling!(execute, self:WBP Discover Tile Base Object Reference=@n1, IsMarqueeScrollingEnabled:bool=true) -> (then)
   > SetPrimaryTextFocused!(execute, self:WBP Discover Tile Base Object Reference=@n1, IsFocused:bool=true) -> (then)
   > On Lock Hover Select!(
    execute
    self:WBP Locked Status Object Reference=Get WBP_LockedStatus() -> (WBP_LockedStatus:WBP Locked Status Object Reference)
    bIsHoverSelected:bool=true
   ) -> (then)
   > PlayAnimationForward(InAnimation=Get Anim_Hover() -> (Anim_Hover:Widget Animation Object Reference))
   > SetVisibility(self=Get Text_Title() -> (Text_Title:Common Text Object Reference), InVisibility=SelfHitTestInvisible)
  ]}
 OnTileSelectionNotActive {nodes [
   OnTileSelectionNotActive() {is=title K2Node_FunctionEntry}
   > Parent: OnTileSelectionNotActive() {is=title K2Node_CallParentFunction}
   n1 = Get WBP_Discover_Tile_Base()
   > UpdatePrimaryTextScrolling(self=@n1, IsMarqueeScrollingEnabled=false)
   > SetPrimaryTextFocused(self=@n1, IsFocused=false)
   > On Lock Hover Select(self=Get WBP_LockedStatus(), bIsHoverSelected=false)
   > PlayAnimationReverse(InAnimation=Get Anim_Hover())
   > SetVisibility(self=Get Text_Title(), InVisibility=Collapsed)
  ]}
 SetIslandName {nodes [
   n1 = SetIslandName!() -> (then, IslandName:Text) {is=title K2Node_FunctionEntry}
   > Parent: SetIslandName!(execute, IslandName:Text=@n1) -> (then) {is=title K2Node_CallParentFunction}
   > SetPrimaryText!(execute, self:WBP Discover Tile Base Object Reference=Get WBP_Discover_Tile_Base(), PrimaryText:Text=@n1) -> (then)
   > "SetText (Text)"(self=Get Text_Title(), InText=@n1)
  ]}
 SetLabelVisiblity {nodes [
   n1 = SetLabelVisiblity!() -> (then, HasUnseenNotification:bool) {is=title K2Node_FunctionEntry}
   > Parent: SetLabelVisiblity!(execute, HasUnseenNotification:bool=@n1) -> (then) {is=title K2Node_CallParentFunction}
   > n2 = Branch(Condition=GetDiscoveryReduxExperimentValue!(FortPlayerController:Player Controller Object Reference=GetOwningPlayer()) -> (ReturnValue:bool))
   > SetVisibility(
    self=Get Highlight_Label() -> (Highlight_Label:WBP UIKit Highlight Object Reference)
    InVisibility=Select(Option 0=Collapsed, Option 1=SelfHitTestInvisible, Index=Get HasUnseenNotification() -> (HasUnseenNotification:bool))
   )
   > Branch(Condition=Get HasUnseenNotification())
   > SetText!(execute, self:WBP UIKit Highlight Base Object Reference=Get Highlight_Label(), Text:Text="LOCTEXT(\"new\")") -> (then)
   SetVisibility(execute=@n2.else, self=Get Highlight_Label(), InVisibility=Collapsed)
  ]}
 SetMobileStyle {nodes [
   SetMobileStyle() {is=title K2Node_FunctionEntry}
   > Branch(Condition=IsMobileGame())
   > SetAutoWrapText!(execute, self:Text Object Reference=Get Text_Title(), InAutoTextWrap:bool=false) -> (then)
  ]}
]
variables []
widgets {
 info {
  inheritedWidgetCount=0
  namedSlotCount=2
  parentClass=&/WBP_IslandTile_Minimal.WBP_IslandTile_Minimal_C
  rootWidgetClass=null
  widgetCount=2
 }
 widgets [
  WBP_LockedStatus {
   widgetClassPath=/DiscoveryBrowser/WBP_LockedStatus
   bIsVariable=true
   widget=WBP_LockedStatus
   properties {lockedReason=Requires Purchase, useNewDesign=true}
  }
  Text_Title {
   widgetClassPath=&CommonTextBlock
   bIsVariable=true
   widget=Text_Title
   properties {
    autoWrapText=true
    mobileFontSizeMultiplier=1.4
    text=Text Block
    textOverflowPolicy=MultilineEllipsis
    visibility=Collapsed
    font {
     bForceMonospaced=false
     bMaterialIsStencil=false
     fontMaterial=null
     fontObject=&HeadingNow
     letterSpacing=0
     monospacedWidth=1
     size=13.5
     skewAmount=0
     typefaceFontName=74 Regular
     outlineSettings {
      bApplyOutlineToDropShadows=false
      bMiteredCorners=false
      bSeparateFillAlpha=false
      outlineMaterial=null
      outlineSize=0
      outlineColor=<rgba 0 0 0 1>
     }
    }
    shadowOffset=<0 0>
   }
  }
 ]
}
